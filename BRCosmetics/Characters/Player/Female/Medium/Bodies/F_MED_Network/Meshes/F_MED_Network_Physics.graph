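# Physics asset for the F_MED_Network female player body, covering the dynamic vest bones at the sides and back.
# Capsules on the pelvis and thighs form the body volumes, with thighs and five small vest base bodies joined to the pelvis by angularly free constraints.
# Each vest flap hangs from its base with hinge-like limits favoring one swing axis.
# Angularly locked constraints chain neighboring flaps from left side across the back to right.
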PhysicsAsset /BRCosmetics/Characters/Player/Female/Medium/Bodies/F_MED_Network/Meshes/F_MED_Network_Physics
bodies {
 dyn_vest_back_1_l [{shapeType=Capsule, length=20, radius=3, shapeName=dyn_vest_back_1_l_capsule, center=<17.3743 -1.9813 2.0883>, rotation=<rot -20.0005 0 0>}]
 dyn_vest_back_1_r [{shapeType=Capsule, length=20, radius=3, shapeName=dyn_vest_back_1_r_capsule, center=<17.7761 -2.5163 -2.7592>, rotation=<rot 14.9999 0 0>}]
 dyn_vest_back_base_c [{shapeType=Capsule, length=0.505, radius=0.505, shapeName=dyn_vest_back_base_c_capsule}]
 dyn_vest_back_base_l [{shapeType=Capsule, length=0.505, radius=0.505, shapeName=dyn_vest_back_base_l_capsule}]
 dyn_vest_back_base_r [{shapeType=Capsule, length=0.505, radius=0.505, shapeName=dyn_vest_back_base_r_capsule}]
 dyn_vest_back_c [{shapeType=Capsule, length=20, radius=3, shapeName=dyn_vest_back_c_capsule, center=<14.3002 -1.3775 -0.0001>}]
 dyn_vest_l_1 [{shapeType=Capsule, length=7, radius=3, shapeName=dyn_vest_l_1_capsule, center=<17.8083 0.7623 -3.1388>, rotation=<rot 0 0 -15.0021>}]
 dyn_vest_l_base [{shapeType=Capsule, length=0.505, radius=0.505, shapeName=dyn_vest_l_base_capsule}]
 dyn_vest_r_1 [{shapeType=Capsule, length=7, radius=3, shapeName=dyn_vest_r_1_capsule, center=<17.8075 -3.7188 5.4844>}]
 dyn_vest_r_base [{shapeType=Capsule, length=0.505, radius=0.505, shapeName=dyn_vest_r_base_capsule}]
 pelvis [{shapeType=Capsule, length=8, radius=10, shapeName=pelvis_capsule, center=<-2.0127 -1.2439 0>}]
 thigh_l [{shapeType=Capsule, length=15, radius=8, shapeName=thigh_l_capsule, center=<-11.8029 0.0002 2.8195>, rotation=<rot 90 0 0>}]
 thigh_r [{shapeType=Capsule, length=15, radius=8, shapeName=thigh_r_capsule, center=<10.988 0.0004 -2.7332>, rotation=<rot 90 0 0>}]
}
constraints [
 {
  bone1Name=thigh_l
  bone2Name=pelvis
  swing1LimitDegrees=45
  swing1Motion=Free
  swing2LimitDegrees=45
  swing2Motion=Free
  twistLimitDegrees=45
  twistMotion=Free
 }
 {
  bone1Name=thigh_r
  bone2Name=pelvis
  swing1LimitDegrees=45
  swing1Motion=Free
  swing2LimitDegrees=45
  swing2Motion=Free
  twistLimitDegrees=45
  twistMotion=Free
 }
 {
  bone1Name=dyn_vest_l_base
  bone2Name=pelvis
  swing1LimitDegrees=45
  swing1Motion=Free
  swing2LimitDegrees=45
  swing2Motion=Free
  twistLimitDegrees=45
  twistMotion=Free
 }
 {
  bone1Name=dyn_vest_back_base_r
  bone2Name=pelvis
  swing1LimitDegrees=45
  swing1Motion=Free
  swing2LimitDegrees=45
  swing2Motion=Free
  twistLimitDegrees=45
  twistMotion=Free
 }
 {
  bone1Name=dyn_vest_back_base_c
  bone2Name=pelvis
  swing1LimitDegrees=45
  swing1Motion=Free
  swing2LimitDegrees=45
  swing2Motion=Free
  twistLimitDegrees=45
  twistMotion=Free
 }
 {
  bone1Name=dyn_vest_back_base_l
  bone2Name=pelvis
  swing1LimitDegrees=45
  swing1Motion=Free
  swing2LimitDegrees=45
  swing2Motion=Free
  twistLimitDegrees=45
  twistMotion=Free
 }
 {
  bone1Name=dyn_vest_r_base
  bone2Name=pelvis
  swing1LimitDegrees=45
  swing1Motion=Free
  swing2LimitDegrees=45
  swing2Motion=Free
  twistLimitDegrees=45
  twistMotion=Free
 }
 {bone1Name=dyn_vest_l_1, bone2Name=dyn_vest_l_base, swing1LimitDegrees=45, swing2LimitDegrees=1, twistLimitDegrees=1}
 {bone1Name=dyn_vest_back_c, bone2Name=dyn_vest_back_base_c, swing1LimitDegrees=45, swing2LimitDegrees=1, twistLimitDegrees=1}
 {bone1Name=dyn_vest_back_1_r, bone2Name=dyn_vest_back_base_r, swing1LimitDegrees=45, swing2LimitDegrees=1, twistLimitDegrees=1}
 {bone1Name=dyn_vest_back_1_l, bone2Name=dyn_vest_back_base_l, swing1LimitDegrees=45, swing2LimitDegrees=1, twistLimitDegrees=1}
 {bone1Name=dyn_vest_r_1, bone2Name=dyn_vest_r_base, swing1LimitDegrees=45, swing2LimitDegrees=1, twistLimitDegrees=1}
 {
  bone1Name=dyn_vest_l_1
  bone2Name=dyn_vest_back_1_l
  swing1LimitDegrees=45
  swing1Motion=Locked
  swing2LimitDegrees=45
  swing2Motion=Locked
  twistLimitDegrees=45
  twistMotion=Locked
 }
 {
  bone1Name=dyn_vest_back_1_l
  bone2Name=dyn_vest_back_c
  swing1LimitDegrees=45
  swing1Motion=Locked
  swing2LimitDegrees=45
  swing2Motion=Locked
  twistLimitDegrees=45
  twistMotion=Locked
 }
 {
  bone1Name=dyn_vest_back_c
  bone2Name=dyn_vest_back_1_r
  swing1LimitDegrees=45
  swing1Motion=Locked
  swing2LimitDegrees=45
  swing2Motion=Locked
  twistLimitDegrees=45
  twistMotion=Locked
 }
 {
  bone1Name=dyn_vest_r_1
  bone2Name=dyn_vest_back_1_r
  swing1LimitDegrees=45
  swing1Motion=Locked
  swing2LimitDegrees=45
  swing2Motion=Locked
  twistLimitDegrees=45
  twistMotion=Locked
 }
]
