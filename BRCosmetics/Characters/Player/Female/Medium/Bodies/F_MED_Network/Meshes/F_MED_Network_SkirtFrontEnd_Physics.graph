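# Physics asset for the skirt of the F_MED_Network female player cosmetic body (SkirtFrontEnd variant), with capsule bodies on pelvis, legs and skirt bones.
# Small master anchors are angularly locked to the pelvis, or right thigh for R2, and lead back, side and R chains with limited swing and tight twist.
# Cross-constraints tie the lower right-side segments to the back-left tip and R chain, while leg joints rotate freely.
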PhysicsAsset /BRCosmetics/Characters/Player/Female/Medium/Bodies/F_MED_Network/Meshes/F_MED_Network_SkirtFrontEnd_Physics
bodies {
 DynSkirt_Right_Side_1 [{shapeType=Capsule, length=15, radius=5, shapeName=DynSkirt_Right_Side_1_capsule, center=<0.9998 -4.7 0>}]
 Dyn_Master_LeftSide [{shapeType=Capsule, length=1, radius=1, shapeName=Dyn_Master_LeftSide_capsule}]
 Dyn_Master_Right_Side [{shapeType=Capsule, length=1, radius=1, shapeName=Dyn_Master_Right_Side_capsule}]
 Dyn_Master_Skirt_B1 [{shapeType=Capsule, length=1, radius=1, shapeName=Dyn_Master_Skirt_B1_capsule}]
 Dyn_Master_Skirt_B_L [{shapeType=Capsule, length=1, radius=1, shapeName=Dyn_Master_Skirt_B_L_capsule}]
 Dyn_Master_Skirt_B_R [{shapeType=Capsule, length=1, radius=1, shapeName=Dyn_Master_Skirt_B_R_capsule}]
 Dyn_Master_Skirt_R2 [{shapeType=Capsule, length=1, radius=1, shapeName=Dyn_Master_Skirt_R2_capsule}]
 Dyn_Skirt_B1 [{shapeType=Capsule, length=15, radius=5, shapeName=Dyn_Skirt_B1_capsule, center=<0.9999 -4 0>}]
 Dyn_Skirt_B2 [{shapeType=Capsule, length=15, radius=5, shapeName=Dyn_Skirt_B2_capsule, center=<0.9998 -6 0>}]
 Dyn_Skirt_B3 [{shapeType=Capsule, length=15, radius=9, shapeName=Dyn_Skirt_B3_capsule, center=<9.1992 -6.5 -0.0003>}]
 Dyn_Skirt_B_L1 [{shapeType=Capsule, length=15, radius=5, shapeName=Dyn_Skirt_B_L1_capsule, center=<0.9999 -4 0>}]
 Dyn_Skirt_B_L2 [{shapeType=Capsule, length=15, radius=5, shapeName=Dyn_Skirt_B_L2_capsule, center=<0.9998 -6 0>}]
 Dyn_Skirt_B_L3 [{shapeType=Capsule, length=20, radius=9, shapeName=Dyn_Skirt_B_L3_capsule, center=<11.5013 -7.855 0>}]
 Dyn_Skirt_B_R1 [{shapeType=Capsule, length=15, radius=5, shapeName=Dyn_Skirt_B_R1_capsule, center=<0.9999 -4 0>}]
 Dyn_Skirt_B_R2 [{shapeType=Capsule, length=15, radius=5, shapeName=Dyn_Skirt_B_R2_capsule, center=<0.9998 -6 0>}]
 Dyn_Skirt_B_R3 [{shapeType=Capsule, length=20, radius=9, shapeName=Dyn_Skirt_B_R3_capsule, center=<6.1183 -9.6706 0>}]
 Dyn_Skirt_Left_1 [{shapeType=Capsule, length=15, radius=5, shapeName=Dyn_Skirt_Left_1_capsule, center=<0.9998 -4.7 0>}]
 Dyn_Skirt_Left_2 [{shapeType=Capsule, length=15, radius=5, shapeName=Dyn_Skirt_Left_2_capsule, center=<0.9998 -6 0>}]
 Dyn_Skirt_R3 [{shapeType=Capsule, length=10, radius=5, shapeName=Dyn_Skirt_R3_capsule, center=<4.8133 -4.3 0>}]
 Dyn_Skirt_R4 [{shapeType=Capsule, length=10, radius=5, shapeName=Dyn_Skirt_R4_capsule, center=<4.8133 -4.3 0>}]
 Dyn_Skirt_RightSide_3 [{shapeType=Capsule, length=20, radius=5, shapeName=Dyn_Skirt_RightSide_3_capsule, center=<1 -6 0>}]
 Dyn_Skirt_RightSide_4 [{shapeType=Capsule, length=15, radius=6, shapeName=Dyn_Skirt_RightSide_4_capsule, center=<3.2424 -6.8 0.0002>}]
 Dyn_Skirt_Right_Side_2 [{shapeType=Capsule, length=15, radius=5, shapeName=Dyn_Skirt_Right_Side_2_capsule, center=<0.9998 -6 0>}]
 calf_l [{shapeType=Capsule, length=40.75, radius=6, shapeName=calf_l_capsule, center=<-30.8432 -1.4742 -0.0001>, rotation=<rot 90 0 0>}]
 calf_r [{shapeType=Capsule, length=40.75, radius=7, shapeName=calf_r_capsule, center=<30.3661 1.5321 1.2987>, rotation=<rot 90 0 0>}]
 pelvis [{shapeType=Capsule, length=6, radius=10, shapeName=pelvis_capsule, center=<0 -0.4801 0>}]
 thigh_l [{shapeType=Capsule, length=25, radius=8, shapeName=thigh_l_capsule, center=<-27.6945 0.0005 2.0967>, rotation=<rot 90 0 0>}]
 thigh_r [{shapeType=Capsule, length=25, radius=9, shapeName=thigh_r_capsule, center=<27.7167 -1.1557 -2.9272>, rotation=<rot 0 94.6877 -89.7904>}]
}
constraints [
 {
  bone1Name=calf_r
  bone2Name=thigh_r
  swing1LimitDegrees=45
  swing1Motion=Free
  swing2LimitDegrees=45
  swing2Motion=Free
  twistLimitDegrees=45
  twistMotion=Free
 }
 {
  bone1Name=thigh_l
  bone2Name=pelvis
  swing1LimitDegrees=45
  swing1Motion=Free
  swing2LimitDegrees=45
  swing2Motion=Free
  twistLimitDegrees=45
  twistMotion=Free
 }
 {
  bone1Name=calf_l
  bone2Name=thigh_l
  swing1LimitDegrees=45
  swing1Motion=Free
  swing2LimitDegrees=45
  swing2Motion=Free
  twistLimitDegrees=45
  twistMotion=Free
 }
 {
  bone1Name=thigh_r
  bone2Name=pelvis
  swing1LimitDegrees=45
  swing1Motion=Free
  swing2LimitDegrees=45
  swing2Motion=Free
  twistLimitDegrees=45
  twistMotion=Free
 }
 {
  bone1Name=Dyn_Master_Skirt_R2
  bone2Name=thigh_r
  swing1LimitDegrees=45
  swing1Motion=Locked
  swing2LimitDegrees=45
  swing2Motion=Locked
  twistLimitDegrees=45
  twistMotion=Locked
 }
 {
  bone1Name=Dyn_Master_Skirt_B1
  bone2Name=pelvis
  swing1LimitDegrees=45
  swing1Motion=Locked
  swing2LimitDegrees=45
  swing2Motion=Locked
  twistLimitDegrees=45
  twistMotion=Locked
 }
 {
  bone1Name=Dyn_Master_Skirt_B_R
  bone2Name=pelvis
  swing1LimitDegrees=45
  swing1Motion=Locked
  swing2LimitDegrees=45
  swing2Motion=Locked
  twistLimitDegrees=45
  twistMotion=Locked
 }
 {
  bone1Name=Dyn_Master_Skirt_B_L
  bone2Name=pelvis
  swing1LimitDegrees=45
  swing1Motion=Locked
  swing2LimitDegrees=45
  swing2Motion=Locked
  twistLimitDegrees=45
  twistMotion=Locked
 }
 {
  bone1Name=Dyn_Master_LeftSide
  bone2Name=pelvis
  swing1LimitDegrees=45
  swing1Motion=Locked
  swing2LimitDegrees=45
  swing2Motion=Locked
  twistLimitDegrees=45
  twistMotion=Locked
 }
 {
  bone1Name=Dyn_Master_Right_Side
  bone2Name=pelvis
  swing1LimitDegrees=45
  swing1Motion=Locked
  swing2LimitDegrees=45
  swing2Motion=Locked
  twistLimitDegrees=45
  twistMotion=Locked
 }
 {bone1Name=Dyn_Skirt_B1, bone2Name=Dyn_Master_Skirt_B1, swing1LimitDegrees=45, swing2LimitDegrees=5, twistLimitDegrees=1}
 {bone1Name=Dyn_Skirt_B_R1, bone2Name=Dyn_Master_Skirt_B_R, swing1LimitDegrees=45, swing2LimitDegrees=5, twistLimitDegrees=1}
 {bone1Name=Dyn_Skirt_B_L1, bone2Name=Dyn_Master_Skirt_B_L, swing1LimitDegrees=45, swing2LimitDegrees=5, twistLimitDegrees=1}
 {bone1Name=Dyn_Skirt_Left_1, bone2Name=Dyn_Master_LeftSide, swing1LimitDegrees=45, swing2LimitDegrees=5, twistLimitDegrees=1}
 {bone1Name=DynSkirt_Right_Side_1, bone2Name=Dyn_Master_Right_Side, swing1LimitDegrees=45, swing2LimitDegrees=5, twistLimitDegrees=1}
 {bone1Name=Dyn_Skirt_B2, bone2Name=Dyn_Skirt_B1, swing1LimitDegrees=40, swing2LimitDegrees=5, twistLimitDegrees=1}
 {bone1Name=Dyn_Skirt_B_R2, bone2Name=Dyn_Skirt_B_R1, swing1LimitDegrees=40, swing2LimitDegrees=5, twistLimitDegrees=1}
 {bone1Name=Dyn_Skirt_B_L2, bone2Name=Dyn_Skirt_B_L1, swing1LimitDegrees=40, swing2LimitDegrees=5, twistLimitDegrees=1}
 {bone1Name=Dyn_Skirt_Left_2, bone2Name=Dyn_Skirt_Left_1, swing1LimitDegrees=40, swing2LimitDegrees=5, twistLimitDegrees=1}
 {bone1Name=Dyn_Skirt_Right_Side_2, bone2Name=DynSkirt_Right_Side_1, swing1LimitDegrees=40, swing2LimitDegrees=5, twistLimitDegrees=1}
 {bone1Name=Dyn_Skirt_B_R3, bone2Name=Dyn_Skirt_B_R2, swing1LimitDegrees=40, swing2LimitDegrees=5, twistLimitDegrees=1}
 {bone1Name=Dyn_Skirt_B_L3, bone2Name=Dyn_Skirt_B_L2, swing1LimitDegrees=40, swing2LimitDegrees=5, twistLimitDegrees=1}
 {bone1Name=Dyn_Skirt_RightSide_3, bone2Name=Dyn_Skirt_Right_Side_2, swing1LimitDegrees=40, swing2LimitDegrees=5, twistLimitDegrees=1}
 {bone1Name=Dyn_Skirt_B3, bone2Name=Dyn_Skirt_B2, swing1LimitDegrees=40, swing2LimitDegrees=5, twistLimitDegrees=1}
 {bone1Name=Dyn_Skirt_RightSide_4, bone2Name=Dyn_Skirt_RightSide_3, swing1LimitDegrees=40, swing2LimitDegrees=5, twistLimitDegrees=1}
 {
  bone1Name=Dyn_Skirt_B_L3
  bone2Name=Dyn_Skirt_RightSide_3
  swing1LimitDegrees=45
  swing1Motion=Locked
  swing2LimitDegrees=45
  swing2Motion=Locked
  twistLimitDegrees=45
  twistMotion=Locked
 }
 {bone1Name=Dyn_Skirt_RightSide_4, bone2Name=Dyn_Skirt_B_L3, swing1LimitDegrees=45, swing2LimitDegrees=5, twistLimitDegrees=1}
 {bone1Name=Dyn_Skirt_R3, bone2Name=Dyn_Master_Skirt_R2, swing1LimitDegrees=40, swing2LimitDegrees=5, twistLimitDegrees=1}
 {bone1Name=Dyn_Skirt_R4, bone2Name=Dyn_Skirt_R3, swing1LimitDegrees=40, swing2LimitDegrees=5, twistLimitDegrees=1}
 {bone1Name=Dyn_Skirt_R3, bone2Name=Dyn_Skirt_RightSide_3, swing1LimitDegrees=20, swing2LimitDegrees=5, twistLimitDegrees=1}
 {bone1Name=Dyn_Skirt_R4, bone2Name=Dyn_Skirt_RightSide_4, swing1LimitDegrees=40, swing2LimitDegrees=5, twistLimitDegrees=1}
]
