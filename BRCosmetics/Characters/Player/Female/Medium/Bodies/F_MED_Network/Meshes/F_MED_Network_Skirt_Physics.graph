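# Physics asset for the skirt of the F_MED_Network female player body, with capsule bodies for skirt bones, pelvis, thighs and calves.
# Tiny master bodies lock angularly to the pelvis, or to the right thigh for R2, and anchor six angle-limited skirt chains.
# Two locked cross-links tie neighboring chains: RightSide_3 to R3 and B_L3 to B3.
# Leg joints swing and twist freely, and several skirt bones carry extra offset capsules.
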PhysicsAsset /BRCosmetics/Characters/Player/Female/Medium/Bodies/F_MED_Network/Meshes/F_MED_Network_Skirt_Physics
bodies {
 DynSkirt_Right_Side_1 [{shapeType=Capsule, length=10, radius=10, shapeName=DynSkirt_Right_Side_1_capsule, center=<11.8374 -11.1 -0.0001>, rotation=<rot 90 0 0>}]
 Dyn_Master_LeftSide [{shapeType=Capsule, length=1, radius=1, shapeName=Dyn_Master_LeftSide_capsule}]
 Dyn_Master_Right_Side [{shapeType=Capsule, length=1, radius=1, shapeName=Dyn_Master_Right_Side_capsule}]
 Dyn_Master_Skirt_B1 [{shapeType=Capsule, length=1, radius=1, shapeName=Dyn_Master_Skirt_B1_capsule}]
 Dyn_Master_Skirt_B_L [{shapeType=Capsule, length=1, radius=1, shapeName=Dyn_Master_Skirt_B_L_capsule}]
 Dyn_Master_Skirt_B_R [{shapeType=Capsule, length=1, radius=1, shapeName=Dyn_Master_Skirt_B_R_capsule}]
 Dyn_Master_Skirt_R2 [{shapeType=Capsule, length=1, radius=1, shapeName=Dyn_Master_Skirt_R2_capsule}]
 Dyn_Skirt_B1 [{shapeType=Capsule, length=10, radius=15, shapeName=Dyn_Skirt_B1_capsule, center=<11.8376 -14.4 -0.0002>, rotation=<rot 90 0 0>}]
 Dyn_Skirt_B2 [{shapeType=Capsule, length=10, radius=15, shapeName=Dyn_Skirt_B2_capsule, center=<11.8376 -12.5 -0.0002>, rotation=<rot 90 0 0>}]
 Dyn_Skirt_B3 [
  {shapeType=Capsule, length=0.5, radius=15, shapeName=Dyn_Skirt_B3_capsule, center=<5.756 -12.5 -0.0002>, rotation=<rot 90 0 0>}
  {shapeType=Capsule, length=0.5, radius=15, shapeName=Dyn_Skirt_B3_capsule, center=<7.1346 -12.5 -13.2908>, rotation=<rot 90 0 0>}
  {shapeType=Capsule, length=0.5, radius=15, shapeName=Dyn_Skirt_B3_capsule, center=<1.9301 -12.5002 12.3643>, rotation=<rot 90 0 0>}
  {shapeType=Capsule, length=15, radius=3, shapeName=Dyn_Skirt_B3_capsule, center=<18.1077 -0.6909 0.1723>, rotation=<rot 15.0004 0 0>}
 ]
 Dyn_Skirt_B_L1 [{shapeType=Capsule, length=10, radius=11, shapeName=Dyn_Skirt_B_L1_capsule, center=<11.8374 -12.6 -0.0002>, rotation=<rot 90 0 0>}]
 Dyn_Skirt_B_L2 [{shapeType=Capsule, length=10, radius=11, shapeName=Dyn_Skirt_B_L2_capsule, center=<11.8374 -12.6 -0.0002>, rotation=<rot 90 0 0>}]
 Dyn_Skirt_B_L3 [
  {shapeType=Capsule, length=15, radius=11, shapeName=Dyn_Skirt_B_L3_capsule, center=<13.872 -12.6 0>, rotation=<rot 90 0 0>}
  {shapeType=Capsule, length=15, radius=3, shapeName=Dyn_Skirt_B_L3_capsule, center=<22.7785 -2.2552 0.0001>, rotation=<rot 19.9999 0 0>}
 ]
 Dyn_Skirt_B_R1 [{shapeType=Capsule, length=10, radius=11, shapeName=Dyn_Skirt_B_R1_capsule, center=<11.8374 -13 0>, rotation=<rot 90 0 0>}]
 Dyn_Skirt_B_R2 [{shapeType=Capsule, length=10, radius=11, shapeName=Dyn_Skirt_B_R2_capsule, center=<11.8374 -13 0>, rotation=<rot 90 0 0>}]
 Dyn_Skirt_B_R3 [
  {shapeType=Capsule, length=0.25, radius=11, shapeName=Dyn_Skirt_B_R3_capsule, center=<4.7798 -12.9999 0>, rotation=<rot 90 0 0>}
  {
   shapeType=Capsule
   length=15
   radius=3
   shapeName=Dyn_Skirt_B_R3_capsule
   center=<14.9082 -2.9113 0.168>
   rotation=<rot 9.6555 -2.613 -15.2223>
  }
 ]
 Dyn_Skirt_Left_1 [{shapeType=Capsule, length=10, radius=6, shapeName=Dyn_Skirt_Left_1_capsule, center=<11.8374 -9.5 -0.0002>, rotation=<rot 90 0 0>}]
 Dyn_Skirt_Left_2 [{shapeType=Capsule, length=10, radius=6, shapeName=Dyn_Skirt_Left_2_capsule, center=<11.8374 -9.5 -0.0002>, rotation=<rot 90 0 0>}]
 Dyn_Skirt_R3 [
  {shapeType=Capsule, length=10, radius=10, shapeName=Dyn_Skirt_R3_capsule, center=<11.8374 -9.5 -0.0002>, rotation=<rot 90 0 0>}
  {shapeType=Capsule, length=10, radius=10, shapeName=Dyn_Skirt_R3_capsule, center=<0.3579 0.2773 13.9606>, rotation=<rot 90 0 0>}
 ]
 Dyn_Skirt_R4 [
  {shapeType=Capsule, length=10, radius=10, shapeName=Dyn_Skirt_R4_capsule, center=<11.8374 -9.5 -0.0002>, rotation=<rot 90 0 0>}
  {shapeType=Capsule, length=10, radius=10, shapeName=Dyn_Skirt_R4_capsule, center=<0.3579 0.2773 13.9606>, rotation=<rot 90 0 0>}
 ]
 Dyn_Skirt_RightSide_3 [
  {shapeType=Capsule, length=10, radius=10, shapeName=Dyn_Skirt_RightSide_3_capsule, center=<11.8374 -10.5 0.0002>, rotation=<rot 90 0 0>}
  {shapeType=Capsule, length=10, radius=10, shapeName=Dyn_Skirt_RightSide_3_capsule, center=<12.2413 -2.699 -19.1159>, rotation=<rot 90 0 0>}
  {shapeType=Capsule, length=10, radius=10, shapeName=Dyn_Skirt_RightSide_3_capsule, center=<11.7629 -9.1157 13.7673>, rotation=<rot 90 0 0>}
 ]
 Dyn_Skirt_RightSide_4 [
  {shapeType=Capsule, length=10, radius=10, shapeName=Dyn_Skirt_RightSide_4_capsule, center=<11.8374 -9.5 0>, rotation=<rot 90 0 0>}
  {shapeType=Capsule, length=10, radius=10, shapeName=Dyn_Skirt_RightSide_4_capsule, center=<12.2413 -2.699 -19.1159>, rotation=<rot 90 0 0>}
 ]
 Dyn_Skirt_Right_Side_2 [
  {shapeType=Capsule, length=10, radius=10, shapeName=Dyn_Skirt_Right_Side_2_capsule, center=<11.8374 -12.5 0>, rotation=<rot 90 0 0>}
  {shapeType=Capsule, length=10, radius=10, shapeName=Dyn_Skirt_Right_Side_2_capsule, center=<12.2413 -2.699 -19.1159>, rotation=<rot 90 0 0>}
 ]
 calf_l [{shapeType=Capsule, length=30, radius=6, shapeName=calf_l_capsule, center=<-17.8703 -4.23 0>, rotation=<rot 90 0 0>}]
 calf_r [{shapeType=Capsule, length=30, radius=7, shapeName=calf_r_capsule, center=<17.5818 2.751 1.2987>, rotation=<rot 90 0 0>}]
 pelvis [{shapeType=Capsule, length=6, radius=10, shapeName=pelvis_capsule, center=<0 -0.4801 0>}]
 thigh_l [{shapeType=Capsule, length=25, radius=8.5, shapeName=thigh_l_capsule, center=<-27.6945 0.0005 2.0967>, rotation=<rot 90 0 0>}]
 thigh_r [{shapeType=Capsule, length=25, radius=10, shapeName=thigh_r_capsule, center=<27.4947 1.5507 -2.9274>, rotation=<rot 0.0183 89.6877 -89.7909>}]
}
constraints [
 {
  bone1Name=calf_r
  bone2Name=thigh_r
  swing1LimitDegrees=45
  swing1Motion=Free
  swing2LimitDegrees=45
  swing2Motion=Free
  twistLimitDegrees=45
  twistMotion=Free
 }
 {
  bone1Name=thigh_l
  bone2Name=pelvis
  swing1LimitDegrees=45
  swing1Motion=Free
  swing2LimitDegrees=45
  swing2Motion=Free
  twistLimitDegrees=45
  twistMotion=Free
 }
 {
  bone1Name=calf_l
  bone2Name=thigh_l
  swing1LimitDegrees=45
  swing1Motion=Free
  swing2LimitDegrees=45
  swing2Motion=Free
  twistLimitDegrees=45
  twistMotion=Free
 }
 {
  bone1Name=thigh_r
  bone2Name=pelvis
  swing1LimitDegrees=45
  swing1Motion=Free
  swing2LimitDegrees=45
  swing2Motion=Free
  twistLimitDegrees=45
  twistMotion=Free
 }
 {
  bone1Name=Dyn_Master_Skirt_R2
  bone2Name=thigh_r
  swing1LimitDegrees=45
  swing1Motion=Locked
  swing2LimitDegrees=45
  swing2Motion=Locked
  twistLimitDegrees=45
  twistMotion=Locked
 }
 {
  bone1Name=Dyn_Master_Skirt_B1
  bone2Name=pelvis
  swing1LimitDegrees=45
  swing1Motion=Locked
  swing2LimitDegrees=45
  swing2Motion=Locked
  twistLimitDegrees=45
  twistMotion=Locked
 }
 {
  bone1Name=Dyn_Master_Skirt_B_R
  bone2Name=pelvis
  swing1LimitDegrees=45
  swing1Motion=Locked
  swing2LimitDegrees=45
  swing2Motion=Locked
  twistLimitDegrees=45
  twistMotion=Locked
 }
 {
  bone1Name=Dyn_Master_Skirt_B_L
  bone2Name=pelvis
  swing1LimitDegrees=45
  swing1Motion=Locked
  swing2LimitDegrees=45
  swing2Motion=Locked
  twistLimitDegrees=45
  twistMotion=Locked
 }
 {
  bone1Name=Dyn_Master_LeftSide
  bone2Name=pelvis
  swing1LimitDegrees=45
  swing1Motion=Locked
  swing2LimitDegrees=45
  swing2Motion=Locked
  twistLimitDegrees=45
  twistMotion=Locked
 }
 {
  bone1Name=Dyn_Master_Right_Side
  bone2Name=pelvis
  swing1LimitDegrees=45
  swing1Motion=Locked
  swing2LimitDegrees=45
  swing2Motion=Locked
  twistLimitDegrees=45
  twistMotion=Locked
 }
 {bone1Name=Dyn_Skirt_R3, bone2Name=Dyn_Master_Skirt_R2, swing1LimitDegrees=40, swing2LimitDegrees=20, twistLimitDegrees=10}
 {bone1Name=Dyn_Skirt_R4, bone2Name=Dyn_Skirt_R3, swing1LimitDegrees=20, swing2LimitDegrees=20, twistLimitDegrees=10}
 {bone1Name=Dyn_Skirt_B1, bone2Name=Dyn_Master_Skirt_B1, swing1LimitDegrees=40, swing2LimitDegrees=20, twistLimitDegrees=10}
 {bone1Name=Dyn_Skirt_B2, bone2Name=Dyn_Skirt_B1, swing1LimitDegrees=40, swing2LimitDegrees=20, twistLimitDegrees=10}
 {bone1Name=Dyn_Skirt_B3, bone2Name=Dyn_Skirt_B2, swing1LimitDegrees=40, swing2LimitDegrees=20, twistLimitDegrees=10}
 {bone1Name=Dyn_Skirt_B_R1, bone2Name=Dyn_Master_Skirt_B_R, swing1LimitDegrees=40, swing2LimitDegrees=20, twistLimitDegrees=10}
 {bone1Name=Dyn_Skirt_B_R3, bone2Name=Dyn_Skirt_B_R2, swing1LimitDegrees=40, swing2LimitDegrees=20, twistLimitDegrees=10}
 {bone1Name=Dyn_Skirt_B_L1, bone2Name=Dyn_Master_Skirt_B_L, swing1LimitDegrees=40, swing2LimitDegrees=20, twistLimitDegrees=10}
 {bone1Name=Dyn_Skirt_B_L2, bone2Name=Dyn_Skirt_B_L1, swing1LimitDegrees=40, swing2LimitDegrees=1, twistLimitDegrees=1}
 {bone1Name=Dyn_Skirt_B_L3, bone2Name=Dyn_Skirt_B_L2, swing1LimitDegrees=40, swing2LimitDegrees=1, twistLimitDegrees=1}
 {bone1Name=Dyn_Skirt_Left_1, bone2Name=Dyn_Master_LeftSide, swing1LimitDegrees=40, swing2LimitDegrees=20, twistLimitDegrees=10}
 {bone1Name=Dyn_Skirt_Left_2, bone2Name=Dyn_Skirt_Left_1, swing1LimitDegrees=40, swing2LimitDegrees=20, twistLimitDegrees=10}
 {bone1Name=DynSkirt_Right_Side_1, bone2Name=Dyn_Master_Right_Side, swing1LimitDegrees=40, swing2LimitDegrees=20, twistLimitDegrees=10}
 {bone1Name=Dyn_Skirt_Right_Side_2, bone2Name=DynSkirt_Right_Side_1, swing1LimitDegrees=40, swing2LimitDegrees=20, twistLimitDegrees=10}
 {bone1Name=Dyn_Skirt_RightSide_3, bone2Name=Dyn_Skirt_Right_Side_2, swing1LimitDegrees=40, swing2LimitDegrees=20, twistLimitDegrees=10}
 {bone1Name=Dyn_Skirt_RightSide_4, bone2Name=Dyn_Skirt_RightSide_3, swing1LimitDegrees=40, swing2LimitDegrees=20, twistLimitDegrees=10}
 {
  bone1Name=Dyn_Skirt_RightSide_3
  bone2Name=Dyn_Skirt_R3
  swing1LimitDegrees=45
  swing1Motion=Locked
  swing2LimitDegrees=45
  swing2Motion=Locked
  twistLimitDegrees=45
  twistMotion=Locked
 }
 {bone1Name=Dyn_Skirt_B_R2, bone2Name=Dyn_Skirt_B_R1, swing1LimitDegrees=60, swing2LimitDegrees=30, twistLimitDegrees=20}
 {
  bone1Name=Dyn_Skirt_B_L3
  bone2Name=Dyn_Skirt_B3
  swing1LimitDegrees=45
  swing1Motion=Locked
  swing2LimitDegrees=45
  swing2Motion=Locked
  twistLimitDegrees=45
  twistMotion=Locked
 }
]
