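# Control Rig for the F_MED_Network female player body that procedurally drives arm and leg twist bones for mesh deformation.
# Upper arms and thighs use swing-twist decomposition about X to counter-rotate their twist bones, partially for arms and fully for thighs, measured against initial pose.
# Forearm and calf twist bones follow hand and foot through weighted offset-preserving constraints, keep only Euler X, and pass half to each second twist bone.
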ControlRigBlueprint /BRCosmetics/Characters/Player/Female/Medium/Bodies/F_MED_Network/Meshes/F_MED_Network__Deform_CtrlRig_twists
graphs [
 RigVMFunctionLibrary {links [], nodes []}
 RigVMModel {nodes [
   RigUnit_BeginExecution = &RigUnit_BeginExecution()
   n1 = /Script/RigVM.RigVMFunction_MathQuaternionToAxisAndAngle(Value=&RigVMFunction_MathQuaternionSwingTwist(TwistAxis=(X=1,Y=0,Z=0), Input=&RigUnit_GetBoneTransform(Bone=upperarm_l, Space=LocalSpace).Transform.Rotation).Twist) {is=struct RigVMUnitNode, template="ToAxisAndAngle::Execute(in Value,out Axis,out Angle)"}
   RerouteNode_2_1 = RigVMDispatchNode(Value=-0.8) {template="DISPATCH_RigVMDispatch_Constant(io Value)"}
   /Script/ControlRig.RigUnit_SetBoneRotation(
    Bone=upperarm_twist_01_l
    Weight=1
    bPropagateToChildren=True
    Rotation=/Script/RigVM.RigVMFunction_MathQuaternionFromAxisAndAngle(
     Angle=&RigVMFunction_MathFloatMul(A=@n1.Angle, B=@RerouteNode_2_1.Value) {template="Multiply::Execute(in A,in B,out Result)"}.Result
     Axis=@n1.Axis
    ) {is=struct RigVMUnitNode, template="FromAxisAndAngle::Execute(in Axis,in Angle,out Result)"}.Result
    ExecutePin=@RigUnit_BeginExecution.ExecutePin
   ) {is=struct RigVMUnitNode}
   n2 = &RigVMFunction_MathQuaternionToAxisAndAngle(Value=&RigVMFunction_MathQuaternionSwingTwist(TwistAxis=(X=1,Y=0,Z=0), Input=&RigUnit_GetBoneTransform(Bone=upperarm_l, Space=LocalSpace).Transform.Rotation).Twist) {template="ToAxisAndAngle::Execute(in Value,out Axis,out Angle)"}
   RerouteNode_2 = RigVMDispatchNode(Value=-0.5) {template="DISPATCH_RigVMDispatch_Constant(io Value)"}
   > &RigUnit_SetBoneRotation(
    Bone=upperarm_twist_02_l
    Weight=1
    bPropagateToChildren=True
    Rotation=&RigVMFunction_MathQuaternionFromAxisAndAngle(
     Angle=&RigVMFunction_MathFloatMul(A=@n2.Angle, B=@RerouteNode_2.Value) {template="Multiply::Execute(in A,in B,out Result)"}.Result
     Axis=@n2.Axis
    ) {template="FromAxisAndAngle::Execute(in Axis,in Angle,out Result)"}.Result
   )
   n3 = &RigVMFunction_MathQuaternionToAxisAndAngle(Value=&RigVMFunction_MathQuaternionSwingTwist(TwistAxis=(X=1,Y=0,Z=0), Input=&RigUnit_GetBoneTransform(Bone=upperarm_r, Space=LocalSpace).Transform.Rotation).Twist) {template="ToAxisAndAngle::Execute(in Value,out Axis,out Angle)"}
   > &RigUnit_SetBoneRotation(
    Bone=upperarm_twist_01_r
    Weight=1
    bPropagateToChildren=True
    Rotation=&RigVMFunction_MathQuaternionFromAxisAndAngle(
     Angle=&RigVMFunction_MathFloatMul(A=@n3.Angle, B=@RerouteNode_2_1.Value) {template="Multiply::Execute(in A,in B,out Result)"}.Result
     Axis=@n3.Axis
    ) {template="FromAxisAndAngle::Execute(in Axis,in Angle,out Result)"}.Result
   )
   n4 = &RigVMFunction_MathQuaternionToAxisAndAngle(Value=&RigVMFunction_MathQuaternionSwingTwist(TwistAxis=(X=1,Y=0,Z=0), Input=&RigUnit_GetBoneTransform(Bone=upperarm_r, Space=LocalSpace).Transform.Rotation).Twist) {template="ToAxisAndAngle::Execute(in Value,out Axis,out Angle)"}
   > &RigUnit_SetBoneRotation(
    Bone=upperarm_twist_02_r
    Weight=1
    bPropagateToChildren=True
    Rotation=&RigVMFunction_MathQuaternionFromAxisAndAngle(
     Angle=&RigVMFunction_MathFloatMul(A=@n4.Angle, B=@RerouteNode_2.Value) {template="Multiply::Execute(in A,in B,out Result)"}.Result
     Axis=@n4.Axis
    ) {template="FromAxisAndAngle::Execute(in Axis,in Angle,out Result)"}.Result
   )
   n5 = &RigUnit_GetBoneTransform(Bone=thigh_l, Space=LocalSpace)
   n6 = /Script/ControlRig.RigUnit_GetInitialBoneTransform(Bone=thigh_l, Space=LocalSpace) {is=struct RigVMUnitNode}
   n7 = &RigVMFunction_MathQuaternionToAxisAndAngle(Value=&RigVMFunction_MathQuaternionSwingTwist(
     TwistAxis=(X=1,Y=0,Z=0)
     Input=&RigVMFunction_MathTransformMakeRelative(Global=@n5.Transform, Parent=@n6.Transform).Local.Rotation
    ).Twist) {template="ToAxisAndAngle::Execute(in Value,out Axis,out Angle)"}
   > &RigUnit_SetBoneRotation(
    Bone=thigh_twist_01_l
    Weight=1
    bPropagateToChildren=True
    Rotation=&RigVMFunction_MathQuaternionFromAxisAndAngle(Angle=&RigVMFunction_MathFloatMul(B=-1, A=@n7.Angle) {template="Multiply::Execute(in A,in B,out Result)"}.Result, Axis=@n7.Axis) {template="FromAxisAndAngle::Execute(in Axis,in Angle,out Result)"}.Result
   )
   n8 = &RigUnit_GetBoneTransform(Bone=thigh_r, Space=LocalSpace)
   RigUnit_GetInitialBoneTransform = &RigUnit_GetInitialBoneTransform(Bone=thigh_r, Space=LocalSpace)
   n9 = &RigVMFunction_MathQuaternionToAxisAndAngle(Value=&RigVMFunction_MathQuaternionSwingTwist(
     TwistAxis=(X=1,Y=0,Z=0)
     Input=&RigVMFunction_MathTransformMakeRelative(Global=@n8.Transform, Parent=@RigUnit_GetInitialBoneTransform.Transform).Local.Rotation
    ).Twist) {template="ToAxisAndAngle::Execute(in Value,out Axis,out Angle)"}
   > &RigUnit_SetBoneRotation(
    Bone=thigh_twist_01_r
    Space=LocalSpace
    Weight=1
    bPropagateToChildren=True
    Rotation=&RigVMFunction_MathQuaternionFromAxisAndAngle(Angle=&RigVMFunction_MathFloatMul(B=-1, A=@n9.Angle) {template="Multiply::Execute(in A,in B,out Result)"}.Result, Axis=@n9.Axis) {template="FromAxisAndAngle::Execute(in Axis,in Angle,out Result)"}.Result
   )
   RigUnit_GetBoneTransform = &RigUnit_GetBoneTransform(Bone=lowerarm_twist_01_l)
   > /Script/ControlRig.RigUnit_TransformConstraint(
    BaseBone=None
    BaseTransform=(Rotation=(X=0,Y=0,Z=0,W=1),Translation=(X=0,Y=0,Z=0),Scale3D=(X=1,Y=1,Z=1))
    BaseTransformSpace=GlobalSpace
    Bone=lowerarm_twist_01_l
    Targets="((Transform=(Rotation=(X=0,Y=0,Z=0,W=1),Translation=(X=0,Y=0,Z=0),Scale3D=(X=1,Y=1,Z=1)),Weight=0.75,bMaintainOffset=True,Filter=(TranslationFilter=(bX=false,bY=false,bZ=false),RotationFilter=(bX=True,bY=false,bZ=false),ScaleFilter=(bX=false,bY=false,bZ=fa ...(truncated)"
    bUseInitialTransforms=True
    Targets.1.Transform=@RigUnit_GetBoneTransform.Transform
    Targets.0.Transform=&RigUnit_GetBoneTransform(Bone=hand_l).Transform
   ) {is=struct RigVMUnitNode}
   n10 = &RigVMFunction_MathQuaternionToEuler(RotationOrder=ZYX, Value=&RigUnit_GetBoneTransform(Bone=lowerarm_twist_01_l, Space=LocalSpace).Transform.Rotation)
   > &RigUnit_SetBoneRotation(
    Bone=lowerarm_twist_01_l
    Weight=1
    bPropagateToChildren=True
    Rotation=&RigVMFunction_MathQuaternionFromEuler(Euler=(X=0,Y=0,Z=0), Euler.X=@n10.Result.X).Result
   )
   > &RigUnit_SetBoneRotation(
    Bone=lowerarm_twist_02_l
    Weight=1
    bPropagateToChildren=True
    Rotation=&RigVMFunction_MathQuaternionFromEuler(
     Euler=(X=0,Y=0,Z=0)
     Euler.X=&RigVMFunction_MathFloatMul(B=0.5, A=@n10.Result.X) {template="Multiply::Execute(in A,in B,out Result)"}.Result
    ).Result
   )
   > &RigUnit_TransformConstraint(
    BaseBone=None
    BaseTransform=(Rotation=(X=0,Y=0,Z=0,W=1),Translation=(X=0,Y=0,Z=0),Scale3D=(X=1,Y=1,Z=1))
    BaseTransformSpace=GlobalSpace
    Bone=lowerarm_twist_01_r
    Targets="((Transform=(Rotation=(X=0,Y=0,Z=0,W=1),Translation=(X=0,Y=0,Z=0),Scale3D=(X=1,Y=1,Z=1)),Weight=0.75,bMaintainOffset=True,Filter=(TranslationFilter=(bX=false,bY=false,bZ=false),RotationFilter=(bX=True,bY=false,bZ=false),ScaleFilter=(bX=false,bY=false,bZ=fa ...(truncated)"
    bUseInitialTransforms=True
    Targets.0.Transform.Rotation=&RigUnit_GetBoneTransform(Bone=hand_r).Transform.Rotation
    Targets.1.Transform.Rotation=&RigUnit_GetBoneTransform(Bone=lowerarm_twist_01_r).Transform.Rotation
   )
   n11 = &RigVMFunction_MathQuaternionToEuler(RotationOrder=ZYX, Value=&RigUnit_GetBoneTransform(Bone=lowerarm_twist_01_r, Space=LocalSpace).Transform.Rotation)
   > RigUnit_SetBoneRotation = &RigUnit_SetBoneRotation(
    Bone=lowerarm_twist_01_r
    Weight=1
    bPropagateToChildren=True
    Rotation=&RigVMFunction_MathQuaternionFromEuler(Euler=(X=0,Y=0,Z=0), Euler.X=@n11.Result.X).Result
   )
   > &RigUnit_SetBoneRotation(
    Bone=lowerarm_twist_02_r
    Weight=1
    bPropagateToChildren=True
    Rotation=&RigVMFunction_MathQuaternionFromEuler(
     Euler=(X=0,Y=0,Z=0)
     Euler.X=&RigVMFunction_MathFloatMul(B=0.5, A=@n11.Result.X) {template="Multiply::Execute(in A,in B,out Result)"}.Result
    ).Result
   )
   > &RigUnit_TransformConstraint(
    BaseBone=None
    BaseTransform=(Rotation=(X=0,Y=0,Z=0,W=1),Translation=(X=0,Y=0,Z=0),Scale3D=(X=1,Y=1,Z=1))
    BaseTransformSpace=GlobalSpace
    Bone=calf_twist_01_l
    Targets="((Transform=(Rotation=(X=0,Y=0,Z=0,W=1),Translation=(X=0,Y=0,Z=0),Scale3D=(X=1,Y=1,Z=1)),Weight=0.75,bMaintainOffset=True,Filter=(TranslationFilter=(bX=false,bY=false,bZ=false),RotationFilter=(bX=True,bY=false,bZ=false),ScaleFilter=(bX=false,bY=false,bZ=fa ...(truncated)"
    bUseInitialTransforms=True
    Targets.0.Transform=&RigUnit_GetBoneTransform(Bone=foot_l).Transform
    Targets.1.Transform=&RigUnit_GetBoneTransform(Bone=calf_twist_01_l).Transform
   )
   n12 = &RigVMFunction_MathQuaternionToEuler(RotationOrder=ZYX, Value=&RigUnit_GetBoneTransform(Bone=calf_twist_01_l, Space=LocalSpace).Transform.Rotation)
   > &RigUnit_SetBoneRotation(
    Bone=calf_twist_01_l
    Weight=1
    bPropagateToChildren=True
    Rotation=&RigVMFunction_MathQuaternionFromEuler(Euler=(X=0,Y=0,Z=0), Euler.X=@n12.Result.X).Result
   )
   > &RigUnit_SetBoneRotation(
    Bone=calf_twist_02_l
    Weight=1
    bPropagateToChildren=True
    Rotation=&RigVMFunction_MathQuaternionFromEuler(
     Euler=(X=0,Y=0,Z=0)
     Euler.X=&RigVMFunction_MathFloatMul(B=0.5, A=@n12.Result.X) {template="Multiply::Execute(in A,in B,out Result)"}.Result
    ).Result
   )
   > &RigUnit_TransformConstraint(
    BaseBone=None
    BaseTransform=(Rotation=(X=0,Y=0,Z=0,W=1),Translation=(X=0,Y=0,Z=0),Scale3D=(X=1,Y=1,Z=1))
    BaseTransformSpace=GlobalSpace
    Bone=calf_twist_01_r
    Targets="((Transform=(Rotation=(X=0,Y=0,Z=0,W=1),Translation=(X=0,Y=0,Z=0),Scale3D=(X=1,Y=1,Z=1)),Weight=0.75,bMaintainOffset=True,Filter=(TranslationFilter=(bX=false,bY=false,bZ=false),RotationFilter=(bX=True,bY=false,bZ=false),ScaleFilter=(bX=false,bY=false,bZ=fa ...(truncated)"
    bUseInitialTransforms=True
    Targets.0.Transform=&RigUnit_GetBoneTransform(Bone=foot_r).Transform
    Targets.1.Transform=&RigUnit_GetBoneTransform(Bone=calf_twist_01_r).Transform
   )
   n13 = &RigVMFunction_MathQuaternionToEuler(RotationOrder=ZYX, Value=&RigUnit_GetBoneTransform(Bone=calf_twist_01_r, Space=LocalSpace).Transform.Rotation)
   > &RigUnit_SetBoneRotation(
    Bone=calf_twist_01_r
    Weight=1
    bPropagateToChildren=True
    Rotation=&RigVMFunction_MathQuaternionFromEuler(Euler=(X=0,Y=0,Z=0), Euler.X=@n13.Result.X).Result
   )
   > &RigUnit_SetBoneRotation(
    Bone=calf_twist_02_r
    Weight=1
    bPropagateToChildren=True
    Rotation=&RigVMFunction_MathQuaternionFromEuler(
     Euler=(X=0,Y=0,Z=0)
     Euler.X=&RigVMFunction_MathFloatMul(B=0.5, A=@n13.Result.X) {template="Multiply::Execute(in A,in B,out Result)"}.Result
    ).Result
   )
  ]}
]
hierarchy [
 root {loc=<0 0 0>, rot=<0 0 0 1>}
 pelvis {parent=root, loc=<0 -0.5301 98.0571>, rot=<0.0127 0.707 -0.0127 -0.707>}
 spine_01 {parent=pelvis, loc=<6.1208 -0.9328 0>, rot=<0 0 0.034 -0.9994>}
 spine_02 {parent=spine_01, loc=<6.5996 -0.2095 0>, rot=<0 0 0.0238 -0.9997>}
 spine_03 {parent=spine_02, loc=<6.4914 -0.1237 0>, rot=<0 0 0.0149 -0.9999>}
 spine_04 {parent=spine_03, loc=<6.9598 0.0794 0>, rot=<0 0 -0.0132 -0.9999>}
 spine_05 {parent=spine_04, loc=<6.9987 0 0>, rot=<0 0 -0.0015 -1>}
 clavicle_l {parent=spine_05, loc=<5.4831 -0.1053 -3.2946>, rot=<0.0465 -0.7289 -0.009 -0.6829>}
 upperarm_l {parent=clavicle_l, loc=<9.8658 -1.2136 -0.4105>, rot=<0.0164 -0.3245 -0.0269 -0.9454>}
 lowerarm_l {parent=upperarm_l, loc=<26.465 -0.8345 0.259>, rot=<0.0036 0.0161 -0.2298 -0.9731>}
 hand_l {parent=lowerarm_l, loc=<22.7355 -0.3203 -0.6105>, rot=<-0.6784 -0.0223 0.0735 -0.7307>}
 index_metacarpal_l {parent=hand_l, loc=<1.6535 0.0183 -2.96>, rot=<0.0158 -0.0124 -0.0173 -0.9996>}
 index_01_l {parent=index_metacarpal_l, loc=<8.3008 0.0357 -0.0524>, rot=<-0.0044 0.0485 0.0924 -0.9945>}
 index_02_l {parent=index_01_l, loc=<3.3587 -0.3004 0.0452>, rot=<-0.002 0.0027 0.3225 -0.9466>}
 index_03_l {parent=index_02_l, loc=<2.5578 -0.123 -0.0252>, rot=<-0.002 0.0067 0.2558 -0.9667>}
 middle_metacarpal_l {parent=hand_l, loc=<2.1856 -0.1725 -0.5984>, rot=<0.006 0.0168 -0.0329 -0.9993>}
 middle_01_l {parent=middle_metacarpal_l, loc=<7.3275 -0.2447 0.0129>, rot=<-0.0208 0.0042 0.0858 -0.9961>}
 middle_02_l {parent=middle_01_l, loc=<4.0123 0.1401 0.1175>, rot=<-0.0218 0.0258 0.3512 -0.9357>}
 middle_03_l {parent=middle_02_l, loc=<2.6077 0.0534 0.1373>, rot=<0.0082 0.0188 0.1708 -0.9851>}
 pinky_metacarpal_l {parent=hand_l, loc=<1.7616 -0.9753 3.3814>, rot=<-0.0397 0.0814 -0.069 -0.9935>}
 pinky_01_l {parent=pinky_metacarpal_l, loc=<7.1709 -0.4831 -0.1094>, rot=<-0.0654 -0.0035 0.2242 -0.9723>}
 pinky_02_l {parent=pinky_01_l, loc=<2.8274 0.0893 -0.1043>, rot=<0.0037 0.0011 0.2854 -0.9584>}
 pinky_03_l {parent=pinky_02_l, loc=<2.2046 0.0307 -0.0078>, rot=<-0.0002 0.0098 0.1445 -0.9895>}
 ring_metacarpal_l {parent=hand_l, loc=<2.2074 -0.3625 1.3638>, rot=<-0.0284 0.0722 -0.0458 -0.9959>}
 ring_01_l {parent=ring_metacarpal_l, loc=<7.0295 -0.6197 -0.1446>, rot=<-0.0489 -0.0252 0.1892 -0.9804>}
 ring_02_l {parent=ring_01_l, loc=<3.6174 0.4647 -0.1027>, rot=<0 0 0.2806 -0.9598>}
 ring_03_l {parent=ring_02_l, loc=<2.4152 0.114 0.0199>, rot=<0 0 0.1189 -0.9929>}
 thumb_01_l {parent=hand_l, loc=<1.3118 -0.3191 -2.459>, rot=<0.6546 -0.4042 0.0664 -0.6354>}
 thumb_02_l {parent=thumb_01_l, loc=<5.9336 -0.0448 0.0401>, rot=<-0.0037 0.0213 0.298 -0.9543>}
 thumb_03_l {parent=thumb_02_l, loc=<2.2223 0.3682 -0.0907>, rot=<0.0118 -0.013 0.0066 -0.9998>}
 weapon_l {parent=hand_l, loc=<-0.1974 -3.4111 0.3811>, rot=<0 0 0 1>}
 lowerarm_twist_01_l {parent=lowerarm_l, loc=<18.316 0 0>, rot=<0 0 0 1>}
 lowerarm_twist_02_l {parent=lowerarm_l, loc=<10.6707 0 0>, rot=<0 0 0 1>}
 deform_elbow_bk_l {parent=lowerarm_l, loc=<-0.583 -1.471 -0.561>, rot=<0 0 0 1>}
 upperarm_twist_01_l {parent=upperarm_l, loc=<7.2135 -0.1134 0>, rot=<0 0 0 1>}
 deform_deltoid_fr_l {parent=upperarm_twist_01_l, loc=<-4.7624 0.1134 0>, rot=<0 0.0007 0.0191 -0.9998>}
 upperarm_twist_02_l {parent=upperarm_l, loc=<19.7998 -0.1134 0>, rot=<0 0 0 1>}
 deform_elbow_fr_l {parent=upperarm_l, loc=<26.5048 0.1875 -0.5816>, rot=<0.0026 0.0163 -0.1662 -0.986>}
 deform_pectoral_fr_l {parent=clavicle_l, loc=<6.2796 0 0>, rot=<0 0 0 1>}
 clavicle_r {parent=spine_05, loc=<5.4831 -0.1053 3.2946>, rot=<0.7289 0.0465 -0.6829 0.009>}
 upperarm_r {parent=clavicle_r, loc=<-9.8658 1.2136 0.4104>, rot=<0.0164 -0.3245 -0.0269 -0.9454>}
 lowerarm_r {parent=upperarm_r, loc=<-26.4651 0.8345 -0.259>, rot=<0.0036 0.0161 -0.2298 -0.9731>}
 hand_r {parent=lowerarm_r, loc=<-22.7356 0.3203 0.6106>, rot=<-0.6784 -0.0223 0.0735 -0.7307>}
 index_metacarpal_r {parent=hand_r, loc=<-1.6534 -0.0181 2.9601>, rot=<0.0158 -0.0124 -0.0173 -0.9996>}
 index_01_r {parent=index_metacarpal_r, loc=<-8.3008 -0.0357 0.0524>, rot=<-0.0044 0.0485 0.0924 -0.9945>}
 index_02_r {parent=index_01_r, loc=<-3.3589 0.3003 -0.0453>, rot=<-0.002 0.0027 0.3225 -0.9466>}
 index_03_r {parent=index_02_r, loc=<-2.5578 0.123 0.0252>, rot=<-0.002 0.0067 0.2558 -0.9667>}
 middle_metacarpal_r {parent=hand_r, loc=<-2.1856 0.1726 0.5984>, rot=<0.006 0.0168 -0.0329 -0.9993>}
 middle_01_r {parent=middle_metacarpal_r, loc=<-7.3273 0.2447 -0.0129>, rot=<-0.0208 0.0042 0.0858 -0.9961>}
 middle_02_r {parent=middle_01_r, loc=<-4.0124 -0.14 -0.1175>, rot=<-0.0218 0.0258 0.3512 -0.9357>}
 middle_03_r {parent=middle_02_r, loc=<-2.6077 -0.0533 -0.1373>, rot=<0.0082 0.0188 0.1708 -0.9851>}
 pinky_metacarpal_r {parent=hand_r, loc=<-1.7617 0.9753 -3.3814>, rot=<-0.0397 0.0814 -0.069 -0.9935>}
 pinky_01_r {parent=pinky_metacarpal_r, loc=<-7.1707 0.4832 0.1094>, rot=<-0.0654 -0.0035 0.2242 -0.9723>}
 pinky_02_r {parent=pinky_01_r, loc=<-2.8274 -0.0892 0.1043>, rot=<0.0037 0.0011 0.2854 -0.9584>}
 pinky_03_r {parent=pinky_02_r, loc=<-2.2045 -0.031 0.0078>, rot=<-0.0002 0.0098 0.1445 -0.9895>}
 ring_metacarpal_r {parent=hand_r, loc=<-2.2073 0.3627 -1.3638>, rot=<-0.0284 0.0722 -0.0458 -0.9959>}
 ring_01_r {parent=ring_metacarpal_r, loc=<-7.0297 0.6197 0.1446>, rot=<-0.0489 -0.0252 0.1892 -0.9804>}
 ring_02_r {parent=ring_01_r, loc=<-3.6171 -0.4648 0.1027>, rot=<0 0 0.2806 -0.9598>}
 ring_03_r {parent=ring_02_r, loc=<-2.4153 -0.1141 -0.0199>, rot=<0 0 0.1189 -0.9929>}
 thumb_01_r {parent=hand_r, loc=<-1.3119 0.3192 2.459>, rot=<0.6546 -0.4042 0.0664 -0.6354>}
 thumb_02_r {parent=thumb_01_r, loc=<-5.9335 0.0447 -0.0401>, rot=<-0.0037 0.0213 0.298 -0.9543>}
 thumb_03_r {parent=thumb_02_r, loc=<-2.2223 -0.3682 0.0907>, rot=<0.0118 -0.013 0.0066 -0.9998>}
 weapon_r {parent=hand_r, loc=<0.1975 3.4112 -0.3811>, rot=<0 0 0 -1>}
 lowerarm_twist_01_r {parent=lowerarm_r, loc=<-18.3161 0 0>, rot=<0 0 0 1>}
 lowerarm_twist_02_r {parent=lowerarm_r, loc=<-10.6707 0 0>, rot=<0 0 0 1>}
 deform_elbow_bk_r {parent=lowerarm_r, loc=<0.5829 1.4707 0.561>, rot=<0 0 0 1>}
 upperarm_twist_01_r {parent=upperarm_r, loc=<-7.2135 0.1134 0>, rot=<0 0 0 1>}
 deform_deltoid_fr_r {parent=upperarm_twist_01_r, loc=<4.7623 -0.1134 0>, rot=<0 0.0007 0.0191 -0.9998>}
 upperarm_twist_02_r {parent=upperarm_r, loc=<-19.7998 0.1134 0>, rot=<0 0 0 1>}
 deform_elbow_fr_r {parent=upperarm_r, loc=<-26.5046 -0.1875 0.5813>, rot=<0.0026 0.0163 -0.1662 -0.986>}
 deform_pectoral_fr_r {parent=clavicle_r, loc=<-6.28 0 0>, rot=<0 0 0 1>}
 neck_01 {parent=spine_05, loc=<11.3138 0.1529 0>, rot=<0 0 -0.2003 -0.9797>}
 neck_02 {parent=neck_01, loc=<4.5627 0 0>, rot=<0 0 0.0504 -0.9987>}
 dyn_vest_l_base {parent=spine_01, loc=<6.0311 5.7578 -9.3769>, rot=<-0.1592 0.4541 0.8745 0.0606>}
 dyn_vest_l_1 {parent=dyn_vest_l_base, loc=<0 0 0>, rot=<0 0 0 1>}
 dyn_vest_back_base_c {parent=spine_01, loc=<9.2364 -7.292 -0.028>, rot=<0.1538 -0.9881 0 0>}
 dyn_vest_back_c {parent=dyn_vest_back_base_c, loc=<-0.4556 0.1589 -0.0241>, rot=<0 0 0 1>}
 dyn_vest_back_base_r {parent=spine_01, loc=<10.0855 -3.5535 9.1387>, rot=<0.1464 -0.8368 0.5275 0.0072>}
 dyn_vest_back_1_r {parent=dyn_vest_back_base_r, loc=<0 0 0>, rot=<0 0 0 1>}
 dyn_vest_back_base_l {parent=spine_01, loc=<8.0961 -3.9008 -10.1226>, rot=<0.1669 -0.8329 -0.5272 -0.0202>}
 dyn_vest_back_1_l {parent=dyn_vest_back_base_l, loc=<-1.7293 0.7715 -0.7035>, rot=<0 0 0 1>}
 dyn_vest_r_base {parent=spine_01, loc=<5.7842 7.8712 9.4148>, rot=<0.1819 -0.5575 0.8079 0.0578>}
 dyn_vest_r_1 {parent=dyn_vest_r_base, loc=<0 0 0>, rot=<0.0098 -0.042 -0.1066 -0.9934>}
 thigh_l {parent=pelvis, loc=<-1.0855 -0.2109 -8.7053>, rot=<0.0009 0.0355 0.0155 -0.9993>}
 calf_l {parent=thigh_l, loc=<-42.5284 -0.0019 -0.7297>, rot=<0.0006 -0.0161 -0.0391 -0.9991>}
 calf_twist_01_l {parent=calf_l, loc=<-35.8483 -1.9197 -0.5794>, rot=<0.0237 0.0095 -0.0035 -0.9997>}
 calf_twist_02_l {parent=calf_l, loc=<-19.1546 -1.7943 -0.2638>, rot=<0.0237 0.0095 -0.0035 -0.9997>}
 foot_l {parent=calf_l, loc=<-46.7401 -2.5014 -0.592>, rot=<0.0427 -0.0238 0.0353 -0.9982>}
 ball_l {parent=foot_l, loc=<-3.5658 10.8066 -0.1544>, rot=<0.0001 -0.0001 -0.7063 -0.7079>}
 deform_knee_bk_01_l {parent=calf_l, loc=<0.6022 -1.0607 0>, rot=<0.0026 0.0216 -0.0407 -0.9989>}
 deform_knee_fr_01_l {parent=calf_l, loc=<-0.2342 0 0>, rot=<0.0017 0.0216 0.0001 -0.9998>}
 thigh_twist_01_l {parent=thigh_l, loc=<-27.952 0 0>, rot=<0.0013 0 0 -1>}
 deform_thigh_fr_l {parent=thigh_twist_01_l, loc=<27.7672 0.2281 -0.0598>, rot=<-0.0013 0 0 -1>}
 deform_thigh_bk_l {parent=thigh_twist_01_l, loc=<23.2101 0.2058 0.1808>, rot=<-0.0013 0 0 -1>}
 Dyn_Master_Skirt_L2 {parent=thigh_twist_01_l, loc=<-1.5022 5.001 -1.2103>, rot=<-0.0526 0.2221 0.9732 -0.029>}
 Dyn_Skirt_L1 {parent=Dyn_Master_Skirt_L2, loc=<0.3558 0.0036 0>, rot=<-0.0235 0 0 -0.9997>}
 Dyn_Skirt_L2 {parent=Dyn_Skirt_L1, loc=<10.6853 0 0>, rot=<0 0 0 1>}
 thigh_r {parent=pelvis, loc=<-1.0853 -0.2109 8.7053>, rot=<0.0355 -0.0009 0.9993 0.0155>}
 calf_r {parent=thigh_r, loc=<42.5286 0.0019 0.7297>, rot=<0.0006 -0.0161 -0.0391 -0.9991>}
 calf_twist_01_r {parent=calf_r, loc=<35.8485 1.9197 0.5793>, rot=<0.0237 0.0095 -0.0035 -0.9997>}
 calf_twist_02_r {parent=calf_r, loc=<19.1546 1.7943 0.2638>, rot=<0.0237 0.0095 -0.0035 -0.9997>}
 foot_r {parent=calf_r, loc=<46.7403 2.5014 0.5919>, rot=<0.0427 -0.0238 0.0353 -0.9982>}
 ball_r {parent=foot_r, loc=<3.5658 -10.8066 0.1544>, rot=<0.0001 -0.0001 -0.7063 -0.7079>}
 deform_knee_bk_01_r {parent=calf_r, loc=<-0.6022 1.0607 0>, rot=<0.0026 0.0216 -0.0407 -0.9989>}
 deform_knee_fr_01_r {parent=calf_r, loc=<0.2341 0 0>, rot=<0.0017 0.0216 0.0001 -0.9998>}
 thigh_twist_01_r {parent=thigh_r, loc=<27.952 0 0>, rot=<0.0013 0 0 -1>}
 deform_thigh_fr_r {parent=thigh_twist_01_r, loc=<-27.767 -0.228 0.06>, rot=<-0.0013 0 0 -1>}
 deform_thigh_bk_r {parent=thigh_twist_01_r, loc=<-23.21 -0.206 -0.181>, rot=<-0.0013 0 0 -1>}
 Dyn_Master_Skirt_R2 {parent=thigh_twist_01_r, loc=<8.7293 -6.9695 3.7091>, rot=<0.2597 0.0324 0.0092 -0.9651>}
 Dyn_Skirt_R3 {parent=Dyn_Master_Skirt_R2, loc=<0.4627 0 0>, rot=<0 0 0 1>}
 Dyn_Skirt_R4 {parent=Dyn_Skirt_R3, loc=<21.4776 0 0>, rot=<0 0 0 1>}
 Dyn_Master_Skirt_B1 {parent=pelvis, loc=<1.3263 -11.6609 -0.019>, rot=<-0.0156 -0.9999 0.0016 0.0025>}
 Dyn_Skirt_B1 {parent=Dyn_Master_Skirt_B1, loc=<0.3384 -0.0001 -0.0001>, rot=<0 0 0.0183 -0.9998>}
 Dyn_Skirt_B2 {parent=Dyn_Skirt_B1, loc=<19.6222 0.1908 0.0227>, rot=<0 0 0 1>}
 Dyn_Skirt_B3 {parent=Dyn_Skirt_B2, loc=<14.6996 0 0>, rot=<0 0 0 1>}
 Dyn_Master_Skirt_B_R {parent=pelvis, loc=<1.9214 -8.8866 -9.2878>, rot=<0.002 -0.9491 -0.3062 -0.0742>}
 Dyn_Skirt_B_R1 {parent=Dyn_Master_Skirt_B_R, loc=<0.2661 -0.0073 0.0111>, rot=<0 -0.0147 0 -0.9999>}
 Dyn_Skirt_B_R2 {parent=Dyn_Skirt_B_R1, loc=<20.8338 0.3678 -0.558>, rot=<0 -0.0171 0 -0.9999>}
 Dyn_Skirt_B_R3 {parent=Dyn_Skirt_B_R2, loc=<13.7767 0 0>, rot=<0 0 0 1>}
 Dyn_Master_Skirt_B_L {parent=pelvis, loc=<1.6356 -7.7877 11.1894>, rot=<0.0155 -0.945 0.3209 0.0613>}
 Dyn_Skirt_B_L1 {parent=Dyn_Master_Skirt_B_L, loc=<0.2661 -0.009 -0.0078>, rot=<0 0 0 1>}
 Dyn_Skirt_B_L2 {parent=Dyn_Skirt_B_L1, loc=<19.439 0.4557 0.3951>, rot=<0 0 0 1>}
 Dyn_Skirt_B_L3 {parent=Dyn_Skirt_B_L2, loc=<15.084 0 0>, rot=<0 0 0 1>}
 Dyn_Master_LeftSide {parent=pelvis, loc=<-6.1182 -1.1752 -14.3048>, rot=<0.033 -0.7075 -0.7038 -0.0548>}
 Dyn_Skirt_Left_1 {parent=Dyn_Master_LeftSide, loc=<0.3584 0 0>, rot=<0 0 0 1>}
 Dyn_Skirt_Left_2 {parent=Dyn_Skirt_Left_1, loc=<18.9506 0 0>, rot=<0 0 0 1>}
 Dyn_Master_Right_Side {parent=pelvis, loc=<-6.0256 -1.1766 13.8343>, rot=<0.0406 -0.7071 0.7032 0.0625>}
 DynSkirt_Right_Side_1 {parent=Dyn_Master_Right_Side, loc=<0.4486 0 0>, rot=<0 0 0 1>}
 Dyn_Skirt_Right_Side_2 {parent=DynSkirt_Right_Side_1, loc=<11.6707 0 0>, rot=<0 0 0 1>}
 Dyn_Skirt_RightSide_3 {parent=Dyn_Skirt_Right_Side_2, loc=<19.3135 0 0>, rot=<0 0 0 1>}
 Dyn_Skirt_RightSide_4 {parent=Dyn_Skirt_RightSide_3, loc=<19.5425 0 0>, rot=<0 0 0 1>}
 attach {parent=root, loc=<0 0 0>, rot=<0 0 0 1>}
 ik_foot_root {parent=root, loc=<0 0 0>, rot=<0 0 0 1>}
 ik_foot_l {parent=ik_foot_root, loc=<14.8262 -7.3545 8.296>, rot=<-0.0357 -0.7045 -0.0284 0.7082>}
 ik_foot_r {parent=ik_foot_root, loc=<-14.8262 -7.3546 8.2958>, rot=<-0.7082 -0.0284 0.7045 -0.0357>}
 ik_hand_root {parent=root, loc=<0 0 0>, rot=<0 0 0 1>}
]
