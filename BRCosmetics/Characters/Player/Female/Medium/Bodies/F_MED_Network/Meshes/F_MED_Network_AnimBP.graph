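# Post-process style cosmetic AnimBP for the F_MED_Network female body, adding skirt and vest secondary motion to the parent mesh's pose.
# Copied pose runs through sculpter and twist Control Rigs, then pelvis-space RigidBody simulations, with an extra front-end-only skirt simulation.
# Blended additive bone offsets then adjust skirt and vest for sprint, crouch, jump, skydive and parachute states.
# EventGraph computes thigh and spine alphas, which the output chain leaves unused.
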AnimBlueprint /BRCosmetics/Characters/Player/Female/Medium/Bodies/F_MED_Network/Meshes/F_MED_Network_AnimBP
target_skeleton=&/F_MED_Network_Skeleton
graphs [
 AnimGraph {nodes [
   n1 = "Transform (Modify) Bone - Bone: Dyn_Master_Skirt_B_R"!(ComponentPose, bAlphaBoolEnabled:bool=True, Translation="0,0,0", Rotation="0,-3,0", Scale="1,1,1") -> (Pose) {
    alphaInputType=Bool
    rotationMode=BMM_Additive
    rotationSpace=BCS_BoneSpace
    scaleMode=BMM_Additive
    scaleSpace=BCS_BoneSpace
    translationMode=BMM_Additive
    translationSpace=BCS_BoneSpace
    alphaBoolBlend {blendOption=Linear}
    alphaScaleBiasClamp {bInterpResult=true}
    boneToModify {boneName=Dyn_Master_Skirt_B_R}
    rotation {yaw=-3}
   }
   "Transform (Modify) Bone - Bone: Dyn_Master_Skirt_B1"!(ComponentPose=@n1, bAlphaBoolEnabled:bool=True, Translation="0,0,0", Rotation="0,-3,0", Scale="1,1,1") -> (Pose) {
    alphaInputType=Bool
    rotationMode=BMM_Additive
    rotationSpace=BCS_BoneSpace
    scaleMode=BMM_Additive
    scaleSpace=BCS_BoneSpace
    translationMode=BMM_Additive
    translationSpace=BCS_BoneSpace
    alphaBoolBlend {blendOption=Linear}
    alphaScaleBiasClamp {bInterpResult=true}
    boneToModify {boneName=Dyn_Master_Skirt_B1}
    rotation {yaw=-3}
   }
   n2 = Control Rig(Source=Copy Pose From Mesh() {bUseAttachedParent=true}) {
    alphaBoolBlend {blendOption=Linear}
    controlRigAssetReference {blueprintRigClass=&/F_MED_Network__Deform_CtrlRig_sculpters.F_MED_Network__Deform_CtrlRig_sculpters_C}
   }
   n3 = Control Rig(Source=@n2) {
    alphaBoolBlend {blendOption=Linear}
    controlRigAssetReference {blueprintRigClass=&/F_MED_Network__Deform_CtrlRig_twists.F_MED_Network__Deform_CtrlRig_twists_C}
   }
   n4 = RigidBody(ComponentPose=Local To Component(LocalPose=@n3)) {
    bClampLinearTranslationLimitToRefPose=true
    bFreezeIncomingPoseOnStart=true
    bOverrideWorldGravity=true
    overridePhysicsAsset=&/F_MED_Network_Skirt_Physics
    simulationSpace=BaseBoneSpace
    alphaBoolBlend {blendOption=Linear}
    baseBoneRef {boneName=pelvis}
    componentAppliedLinearAccClamp=<2000 2000 2000>
    componentLinearAccScale=<0.1 0.1 0.1>
    componentLinearVelScale=<0.1 0.1 0.1>
    overrideWorldGravity {z=-980}
   }
   n5 = "Transform (Modify) Bone - Bone: Dyn_Master_Skirt_B1"!(ComponentPose=@n4, Alpha:float=1, Translation="0,0,0", Rotation="0,-3,0", Scale="1,1,1") -> (Pose) {
    rotationMode=BMM_Additive
    rotationSpace=BCS_BoneSpace
    scaleMode=BMM_Additive
    scaleSpace=BCS_BoneSpace
    translationMode=BMM_Additive
    translationSpace=BCS_BoneSpace
    alphaBoolBlend {blendOption=Linear}
    alphaScaleBiasClamp {bInterpResult=true}
    boneToModify {boneName=Dyn_Master_Skirt_B1}
    rotation {yaw=-3}
   }
   n6 = "Transform (Modify) Bone - Bone: Dyn_Master_Skirt_B_L"(ComponentPose=@n5, Alpha=1, Translation="0,0,0", Rotation="0,-3,0", Scale="1,1,1") {
    rotationMode=BMM_Additive
    rotationSpace=BCS_BoneSpace
    scaleMode=BMM_Additive
    scaleSpace=BCS_BoneSpace
    translationMode=BMM_Additive
    translationSpace=BCS_BoneSpace
    alphaBoolBlend {blendOption=Linear}
    alphaScaleBiasClamp {bInterpResult=true}
    boneToModify {boneName=Dyn_Master_Skirt_B_L}
    rotation {yaw=-3}
   }
   n7 = "Transform (Modify) Bone - Bone: Dyn_Master_Skirt_B_R"!(ComponentPose=@n6, Alpha:float=1, Translation="0,0,0", Rotation="0,-3,0", Scale="1,1,1") -> (Pose) {
    rotationMode=BMM_Additive
    rotationSpace=BCS_BoneSpace
    scaleMode=BMM_Additive
    scaleSpace=BCS_BoneSpace
    translationMode=BMM_Additive
    translationSpace=BCS_BoneSpace
    alphaBoolBlend {blendOption=Linear}
    alphaScaleBiasClamp {bInterpResult=true}
    boneToModify {boneName=Dyn_Master_Skirt_B_R}
    rotation {yaw=-3}
   }
   n8 = "Transform (Modify) Bone - Bone: Dyn_Master_LeftSide"(ComponentPose=@n7, Alpha=1, Translation="0,0,0", Rotation="0,0,0", Scale="1,1,1") {
    rotationMode=BMM_Additive
    rotationSpace=BCS_BoneSpace
    scaleMode=BMM_Replace
    scaleSpace=BCS_BoneSpace
    translationMode=BMM_Additive
    translationSpace=BCS_BoneSpace
    alphaBoolBlend {blendInTime=0.2, blendOption=Linear, blendOutTime=0.1}
    boneToModify {boneName=Dyn_Master_LeftSide}
   }
   n9 = RigidBody(ComponentPose=@n8) {
    bClampLinearTranslationLimitToRefPose=true
    bFreezeIncomingPoseOnStart=true
    bOverrideWorldGravity=true
    overridePhysicsAsset=&/F_MED_Network_Physics
    simulationSpace=BaseBoneSpace
    alphaBoolBlend {blendOption=Linear}
    baseBoneRef {boneName=pelvis}
    componentAppliedLinearAccClamp=<2000 2000 2000>
    componentLinearAccScale=<0.1 0.1 0.1>
    componentLinearVelScale=<0.1 0.1 0.1>
    overrideWorldGravity {z=-980}
   }
   n10 = RigidBody!(ComponentPose=@n9, bAlphaBoolEnabled:bool=Get bIsInFrontEnd(), ExternalForce) -> (Pose) {
    alphaInputType=Bool
    bClampLinearTranslationLimitToRefPose=true
    bFreezeIncomingPoseOnStart=true
    bOverrideWorldGravity=true
    overridePhysicsAsset=&/F_MED_Network_SkirtFrontEnd_Physics
    simulationSpace=BaseBoneSpace
    alphaBoolBlend {blendOption=Linear}
    baseBoneRef {boneName=pelvis}
    componentAppliedLinearAccClamp=<2000 2000 2000>
    componentLinearAccScale=<0.1 0.1 0.1>
    componentLinearVelScale=<0.1 0.1 0.1>
    overrideWorldGravity {z=-980}
   }
   n11 = Get bIsSprinting()
   n12 = "Transform (Modify) Bone - Bone: Dyn_Master_Skirt_B1"!(ComponentPose=@n10, bAlphaBoolEnabled:bool=@n11, Translation="0,0,0", Rotation="0,-6,0", Scale="0.9,0.9,0.9") -> (Pose) {
    alphaInputType=Bool
    rotationMode=BMM_Additive
    rotationSpace=BCS_BoneSpace
    scaleMode=BMM_Additive
    scaleSpace=BCS_BoneSpace
    translationMode=BMM_Additive
    translationSpace=BCS_BoneSpace
    alphaBoolBlend {blendInTime=1, blendOption=Linear, blendOutTime=1}
    boneToModify {boneName=Dyn_Master_Skirt_B1}
    rotation {yaw=-6}
    scale=<0.9 0.9 0.9>
   }
   n13 = "Transform (Modify) Bone - Bone: Dyn_Master_Skirt_B_R"!(ComponentPose=@n12, bAlphaBoolEnabled:bool=@n11, Translation="0,0,0", Rotation="0,-5,-20", Scale="0.9,0.9,0.9") -> (Pose) {
    alphaInputType=Bool
    rotationMode=BMM_Additive
    rotationSpace=BCS_BoneSpace
    scaleMode=BMM_Additive
    scaleSpace=BCS_BoneSpace
    translationMode=BMM_Additive
    translationSpace=BCS_BoneSpace
    alphaBoolBlend {blendInTime=1, blendOption=Linear, blendOutTime=1}
    boneToModify {boneName=Dyn_Master_Skirt_B_R}
    rotation {roll=-20, yaw=-5}
    scale=<0.9 0.9 0.9>
   }
   n14 = "Transform (Modify) Bone - Bone: Dyn_Master_Skirt_B_L"!(ComponentPose=@n13, bAlphaBoolEnabled:bool=@n11, Translation="0,0,0", Rotation="0,-6,0", Scale="0.8,0.8,0.8") -> (Pose) {
    alphaInputType=Bool
    rotationMode=BMM_Additive
    rotationSpace=BCS_BoneSpace
    scaleMode=BMM_Additive
    scaleSpace=BCS_BoneSpace
    translationMode=BMM_Additive
    translationSpace=BCS_BoneSpace
    alphaBoolBlend {blendInTime=1, blendOption=Linear, blendOutTime=1}
    boneToModify {boneName=Dyn_Master_Skirt_B_L}
    rotation {yaw=-6}
    scale=<0.8 0.8 0.8>
   }
   n15 = "Transform (Modify) Bone - Bone: Dyn_Master_LeftSide"!(ComponentPose=@n14, bAlphaBoolEnabled:bool=@n11, Translation="0,0,0", Rotation="0,5,0", Scale="1,1,1") -> (Pose) {
    alphaInputType=Bool
    rotationMode=BMM_Additive
    rotationSpace=BCS_BoneSpace
    scaleMode=BMM_Replace
    scaleSpace=BCS_BoneSpace
    translationMode=BMM_Additive
    translationSpace=BCS_BoneSpace
    alphaBoolBlend {blendInTime=0.2, blendOption=Linear, blendOutTime=0.1}
    boneToModify {boneName=Dyn_Master_LeftSide}
    rotation {yaw=5}
   }
   n16 = Get bIsCrouching()
   n17 = "Transform (Modify) Bone - Bone: Dyn_Master_Skirt_B1"(ComponentPose=@n15, bAlphaBoolEnabled=@n16, Translation="0,0,0", Rotation="0,-2,0", Scale="0.75,1,1") {
    alphaInputType=Bool
    rotationMode=BMM_Additive
    rotationSpace=BCS_BoneSpace
    scaleMode=BMM_Replace
    scaleSpace=BCS_BoneSpace
    translationMode=BMM_Additive
    translationSpace=BCS_BoneSpace
    alphaBoolBlend {blendInTime=0.2, blendOption=Linear, blendOutTime=0.1}
    boneToModify {boneName=Dyn_Master_Skirt_B1}
    rotation {yaw=-2}
    scale {x=0.75}
   }
   n18 = "Transform (Modify) Bone - Bone: Dyn_Master_Skirt_B_R"(ComponentPose=@n17, bAlphaBoolEnabled=@n16, Translation="0,0,0", Rotation="0,-3,0", Scale="0.75,1,1") {
    alphaInputType=Bool
    rotationMode=BMM_Additive
    rotationSpace=BCS_BoneSpace
    scaleMode=BMM_Replace
    scaleSpace=BCS_BoneSpace
    translationMode=BMM_Additive
    translationSpace=BCS_BoneSpace
    alphaBoolBlend {blendInTime=0.2, blendOption=Linear, blendOutTime=0.1}
    boneToModify {boneName=Dyn_Master_Skirt_B_R}
    rotation {yaw=-3}
    scale {x=0.75}
   }
   n19 = "Transform (Modify) Bone - Bone: Dyn_Master_Skirt_B_L"(ComponentPose=@n18, bAlphaBoolEnabled=@n16, Translation="0,0,0", Rotation="0,-1,0", Scale="0.5,0.7,0.7") {
    alphaInputType=Bool
    rotationMode=BMM_Additive
    rotationSpace=BCS_BoneSpace
    scaleMode=BMM_Replace
    scaleSpace=BCS_BoneSpace
    translationMode=BMM_Additive
    translationSpace=BCS_BoneSpace
    alphaBoolBlend {blendInTime=0.2, blendOption=Linear, blendOutTime=0.1}
    boneToModify {boneName=Dyn_Master_Skirt_B_L}
    rotation {yaw=-1}
    scale=<0.5 0.7 0.7>
   }
   n20 = "Transform (Modify) Bone - Bone: Dyn_Master_Right_Side"!(ComponentPose=@n19, bAlphaBoolEnabled:bool=@n16, Translation="0,-3,0", Rotation="0,-5,0", Scale="0.7,1,1") -> (Pose) {
    alphaInputType=Bool
    rotationMode=BMM_Additive
    rotationSpace=BCS_BoneSpace
    scaleMode=BMM_Replace
    scaleSpace=BCS_BoneSpace
    translationMode=BMM_Additive
    translationSpace=BCS_BoneSpace
    alphaBoolBlend {blendInTime=0.2, blendOption=Linear, blendOutTime=0.1}
    boneToModify {boneName=Dyn_Master_Right_Side}
    rotation {yaw=-5}
    scale {x=0.7}
    translation {y=-3}
   }
   n21 = "Transform (Modify) Bone - Bone: Dyn_Master_LeftSide"(ComponentPose=@n20, bAlphaBoolEnabled=@n16, Translation="0,-5,0", Rotation="50,-20,-30", Scale="0.7,1,1") {
    alphaInputType=Bool
    rotationMode=BMM_Additive
    rotationSpace=BCS_BoneSpace
    scaleMode=BMM_Replace
    scaleSpace=BCS_BoneSpace
    translationMode=BMM_Additive
    translationSpace=BCS_BoneSpace
    alphaBoolBlend {blendInTime=0.2, blendOption=Linear, blendOutTime=0.1}
    boneToModify {boneName=Dyn_Master_LeftSide}
    rotation=<rot 50 -20 -30>
    scale {x=0.7}
    translation {y=-5}
   }
   n22 = "Transform (Modify) Bone - Bone: dyn_vest_back_base_c"(ComponentPose=@n21, Alpha=1, Translation="0,-1,0", Rotation="0,-10,0", Scale="1,1,1") {
    rotationMode=BMM_Additive
    rotationSpace=BCS_BoneSpace
    scaleMode=BMM_Replace
    scaleSpace=BCS_BoneSpace
    translationMode=BMM_Additive
    translationSpace=BCS_BoneSpace
    alphaBoolBlend {blendInTime=0.2, blendOption=Linear, blendOutTime=0.1}
    boneToModify {boneName=dyn_vest_back_base_c}
    rotation {yaw=-10}
    translation {y=-1}
   }
   n23 = "Transform (Modify) Bone - Bone: dyn_vest_back_base_r"(ComponentPose=@n22, Alpha=1, Translation="0,-1,0", Rotation="0,0,0", Scale="1,1,1") {
    rotationMode=BMM_Additive
    rotationSpace=BCS_BoneSpace
    scaleMode=BMM_Replace
    scaleSpace=BCS_BoneSpace
    translationMode=BMM_Additive
    translationSpace=BCS_BoneSpace
    alphaBoolBlend {blendInTime=0.2, blendOption=Linear, blendOutTime=0.1}
    boneToModify {boneName=dyn_vest_back_base_r}
    translation {y=-1}
   }
   n24 = Get bIsCrouching()
   n25 = "Transform (Modify) Bone - Bone: dyn_vest_l_base"!(ComponentPose=@n23, bAlphaBoolEnabled:bool=@n24, Translation="-3,3.5,-7", Rotation="0,0,-30", Scale="1,1,1") -> (Pose) {
    alphaInputType=Bool
    rotationMode=BMM_Additive
    rotationSpace=BCS_BoneSpace
    scaleMode=BMM_Replace
    scaleSpace=BCS_BoneSpace
    translationMode=BMM_Additive
    translationSpace=BCS_BoneSpace
    alphaBoolBlend {blendInTime=0.2, blendOption=Linear, blendOutTime=0.1}
    boneToModify {boneName=dyn_vest_l_base}
    rotation {roll=-30}
    translation=<-3 3.5 -7>
   }
   n26 = "Transform (Modify) Bone - Bone: dyn_vest_r_base"!(ComponentPose=@n25, bAlphaBoolEnabled:bool=@n24, Translation="0,0,4", Rotation="0,0,10", Scale="1,1,1") -> (Pose) {
    alphaInputType=Bool
    rotationMode=BMM_Additive
    rotationSpace=BCS_BoneSpace
    scaleMode=BMM_Replace
    scaleSpace=BCS_BoneSpace
    translationMode=BMM_Additive
    translationSpace=BCS_BoneSpace
    alphaBoolBlend {blendInTime=0.2, blendOption=Linear, blendOutTime=0.1}
    boneToModify {boneName=dyn_vest_r_base}
    rotation {roll=10}
    translation {z=4}
   }
   n27 = Get bIsJumping()
   n28 = "Transform (Modify) Bone - Bone: dyn_vest_back_base_c"!(ComponentPose=@n26, bAlphaBoolEnabled:bool=@n27, Translation="0,0,0", Rotation="0,-20,0", Scale="1,1,1") -> (Pose) {
    alphaInputType=Bool
    rotationMode=BMM_Additive
    rotationSpace=BCS_BoneSpace
    scaleMode=BMM_Replace
    scaleSpace=BCS_BoneSpace
    translationMode=BMM_Additive
    translationSpace=BCS_BoneSpace
    alphaBoolBlend {blendInTime=0.2, blendOption=Linear, blendOutTime=1}
    boneToModify {boneName=dyn_vest_back_base_c}
    rotation {yaw=-20}
   }
   n29 = "Transform (Modify) Bone - Bone: dyn_vest_back_base_r"!(ComponentPose=@n28, bAlphaBoolEnabled:bool=@n27, Translation="0,0,0", Rotation="0,-10,0", Scale="1,1,1") -> (Pose) {
    alphaInputType=Bool
    rotationMode=BMM_Additive
    rotationSpace=BCS_BoneSpace
    scaleMode=BMM_Replace
    scaleSpace=BCS_BoneSpace
    translationMode=BMM_Additive
    translationSpace=BCS_BoneSpace
    alphaBoolBlend {blendInTime=0.2, blendOption=Linear, blendOutTime=1}
    boneToModify {boneName=dyn_vest_back_base_r}
    rotation {yaw=-10}
   }
   n30 = "Transform (Modify) Bone - Bone: dyn_vest_back_base_l"!(ComponentPose=@n29, bAlphaBoolEnabled:bool=@n27, Translation="0,0,0", Rotation="0,-10,0", Scale="1,1,1") -> (Pose) {
    alphaInputType=Bool
    rotationMode=BMM_Additive
    rotationSpace=BCS_BoneSpace
    scaleMode=BMM_Replace
    scaleSpace=BCS_BoneSpace
    translationMode=BMM_Additive
    translationSpace=BCS_BoneSpace
    alphaBoolBlend {blendInTime=0.2, blendOption=Linear, blendOutTime=1}
    boneToModify {boneName=dyn_vest_back_base_l}
    rotation {yaw=-10}
   }
   n31 = Get bIsSkyDiving()
   n32 = "Transform (Modify) Bone - Bone: Dyn_Master_Skirt_B1"(ComponentPose=@n30, bAlphaBoolEnabled=@n31, Translation="0,0,0", Rotation="0,-10,0", Scale="0.9,0.9,0.9") {
    alphaInputType=Bool
    rotationMode=BMM_Additive
    rotationSpace=BCS_BoneSpace
    scaleMode=BMM_Replace
    scaleSpace=BCS_BoneSpace
    translationMode=BMM_Additive
    translationSpace=BCS_BoneSpace
    alphaBoolBlend {blendInTime=0.2, blendOption=Linear, blendOutTime=0.1}
    boneToModify {boneName=Dyn_Master_Skirt_B1}
    rotation {yaw=-10}
    scale=<0.9 0.9 0.9>
   }
   n33 = "Transform (Modify) Bone - Bone: Dyn_Master_Skirt_B_R"(ComponentPose=@n32, bAlphaBoolEnabled=@n31, Translation="0,0,0", Rotation="0,-10,0", Scale="0.9,0.9,0.9") {
    alphaInputType=Bool
    rotationMode=BMM_Additive
    rotationSpace=BCS_BoneSpace
    scaleMode=BMM_Replace
    scaleSpace=BCS_BoneSpace
    translationMode=BMM_Additive
    translationSpace=BCS_BoneSpace
    alphaBoolBlend {blendInTime=0.2, blendOption=Linear, blendOutTime=0.1}
    boneToModify {boneName=Dyn_Master_Skirt_B_R}
    rotation {yaw=-10}
    scale=<0.9 0.9 0.9>
   }
   n34 = "Transform (Modify) Bone - Bone: Dyn_Master_Skirt_B_L"(ComponentPose=@n33, bAlphaBoolEnabled=@n31, Translation="0,0,0", Rotation="0,-10,0", Scale="0.9,0.9,0.9") {
    alphaInputType=Bool
    rotationMode=BMM_Additive
    rotationSpace=BCS_BoneSpace
    scaleMode=BMM_Replace
    scaleSpace=BCS_BoneSpace
    translationMode=BMM_Additive
    translationSpace=BCS_BoneSpace
    alphaBoolBlend {blendInTime=0.2, blendOption=Linear, blendOutTime=0.1}
    boneToModify {boneName=Dyn_Master_Skirt_B_L}
    rotation {yaw=-10}
    scale=<0.9 0.9 0.9>
   }
   n35 = "Transform (Modify) Bone - Bone: Dyn_Master_Right_Side"(ComponentPose=@n34, bAlphaBoolEnabled=@n31, Translation="0,0,0", Rotation="0,0,0", Scale="1,1,1") {
    alphaInputType=Bool
    rotationMode=BMM_Additive
    rotationSpace=BCS_BoneSpace
    scaleMode=BMM_Replace
    scaleSpace=BCS_BoneSpace
    translationMode=BMM_Additive
    translationSpace=BCS_BoneSpace
    alphaBoolBlend {blendInTime=0.2, blendOption=Linear, blendOutTime=0.1}
    boneToModify {boneName=Dyn_Master_Right_Side}
   }
   n36 = "Transform (Modify) Bone - Bone: Dyn_Master_LeftSide"(ComponentPose=@n35, bAlphaBoolEnabled=@n31, Translation="0,0,0", Rotation="0,0,0", Scale="1,1,1") {
    alphaInputType=Bool
    rotationMode=BMM_Additive
    rotationSpace=BCS_BoneSpace
    scaleMode=BMM_Replace
    scaleSpace=BCS_BoneSpace
    translationMode=BMM_Additive
    translationSpace=BCS_BoneSpace
    alphaBoolBlend {blendInTime=0.2, blendOption=Linear, blendOutTime=0.1}
    boneToModify {boneName=Dyn_Master_LeftSide}
   }
   n37 = "Transform (Modify) Bone - Bone: dyn_vest_back_base_r"(ComponentPose=@n36, bAlphaBoolEnabled=@n31, Translation="0,0,0", Rotation="0,-5,0", Scale="1,1,1") {
    alphaInputType=Bool
    rotationMode=BMM_Additive
    rotationSpace=BCS_BoneSpace
    scaleMode=BMM_Replace
    scaleSpace=BCS_BoneSpace
    translationMode=BMM_Additive
    translationSpace=BCS_BoneSpace
    alphaBoolBlend {blendInTime=0.2, blendOption=Linear, blendOutTime=0.1}
    boneToModify {boneName=dyn_vest_back_base_r}
    rotation {yaw=-5}
   }
   n38 = "Transform (Modify) Bone - Bone: dyn_vest_back_base_l"(ComponentPose=@n37, bAlphaBoolEnabled=@n31, Translation="0,0,0", Rotation="0,-5,0", Scale="1,1,1") {
    alphaInputType=Bool
    rotationMode=BMM_Additive
    rotationSpace=BCS_BoneSpace
    scaleMode=BMM_Replace
    scaleSpace=BCS_BoneSpace
    translationMode=BMM_Additive
    translationSpace=BCS_BoneSpace
    alphaBoolBlend {blendInTime=0.2, blendOption=Linear, blendOutTime=0.1}
    boneToModify {boneName=dyn_vest_back_base_l}
    rotation {yaw=-5}
   }
   n39 = "Transform (Modify) Bone - Bone: dyn_vest_back_base_c"(ComponentPose=@n38, bAlphaBoolEnabled=@n31, Translation="0,0,0", Rotation="0,-1,0", Scale="1,1,1") {
    alphaInputType=Bool
    rotationMode=BMM_Additive
    rotationSpace=BCS_BoneSpace
    scaleMode=BMM_Replace
    scaleSpace=BCS_BoneSpace
    translationMode=BMM_Additive
    translationSpace=BCS_BoneSpace
    alphaBoolBlend {blendInTime=0.2, blendOption=Linear, blendOutTime=0.1}
    boneToModify {boneName=dyn_vest_back_base_c}
    rotation {yaw=-1}
   }
   n40 = Get bIsParachuteOpened()
   n41 = "Transform (Modify) Bone - Bone: Dyn_Master_Skirt_B1"(ComponentPose=@n39, bAlphaBoolEnabled=@n40, Translation="0,0,0", Rotation="0,-10,0", Scale="1,1,1") {
    alphaInputType=Bool
    rotationMode=BMM_Additive
    rotationSpace=BCS_BoneSpace
    scaleMode=BMM_Replace
    scaleSpace=BCS_BoneSpace
    translationMode=BMM_Additive
    translationSpace=BCS_BoneSpace
    alphaBoolBlend {blendInTime=0.2, blendOption=Linear, blendOutTime=2}
    boneToModify {boneName=Dyn_Master_Skirt_B1}
    rotation {yaw=-10}
   }
   n42 = "Transform (Modify) Bone - Bone: Dyn_Master_Skirt_B_R"(ComponentPose=@n41, bAlphaBoolEnabled=@n40, Translation="0,0,0", Rotation="0,-30,0", Scale="1,1,1") {
    alphaInputType=Bool
    rotationMode=BMM_Additive
    rotationSpace=BCS_BoneSpace
    scaleMode=BMM_Replace
    scaleSpace=BCS_BoneSpace
    translationMode=BMM_Additive
    translationSpace=BCS_BoneSpace
    alphaBoolBlend {blendInTime=0.2, blendOption=Linear, blendOutTime=2}
    boneToModify {boneName=Dyn_Master_Skirt_B_R}
    rotation {yaw=-30}
   }
   n43 = "Transform (Modify) Bone - Bone: Dyn_Master_Skirt_B_L"(ComponentPose=@n42, bAlphaBoolEnabled=@n40, Translation="0,0,0", Rotation="0,-10,0", Scale="1,1,1") {
    alphaInputType=Bool
    rotationMode=BMM_Additive
    rotationSpace=BCS_BoneSpace
    scaleMode=BMM_Replace
    scaleSpace=BCS_BoneSpace
    translationMode=BMM_Additive
    translationSpace=BCS_BoneSpace
    alphaBoolBlend {blendInTime=0.2, blendOption=Linear, blendOutTime=2}
    boneToModify {boneName=Dyn_Master_Skirt_B_L}
    rotation {yaw=-10}
   }
   n44 = "Transform (Modify) Bone - Bone: Dyn_Master_Right_Side"(ComponentPose=@n43, bAlphaBoolEnabled=@n40, Translation="0,0,0", Rotation="0,0,0", Scale="1,1,1") {
    alphaInputType=Bool
    rotationMode=BMM_Additive
    rotationSpace=BCS_BoneSpace
    scaleMode=BMM_Replace
    scaleSpace=BCS_BoneSpace
    translationMode=BMM_Additive
    translationSpace=BCS_BoneSpace
    alphaBoolBlend {blendInTime=0.2, blendOption=Linear, blendOutTime=2}
    boneToModify {boneName=Dyn_Master_Right_Side}
   }
   n45 = "Transform (Modify) Bone - Bone: Dyn_Master_LeftSide"(ComponentPose=@n44, bAlphaBoolEnabled=@n40, Translation="0,0,0", Rotation="0,0,0", Scale="1,1,1") {
    alphaInputType=Bool
    rotationMode=BMM_Additive
    rotationSpace=BCS_BoneSpace
    scaleMode=BMM_Replace
    scaleSpace=BCS_BoneSpace
    translationMode=BMM_Additive
    translationSpace=BCS_BoneSpace
    alphaBoolBlend {blendInTime=0.2, blendOption=Linear, blendOutTime=2}
    boneToModify {boneName=Dyn_Master_LeftSide}
   }
   n46 = "Transform (Modify) Bone - Bone: dyn_vest_back_base_r"(ComponentPose=@n45, bAlphaBoolEnabled=@n40, Translation="0,0,0", Rotation="0,-5,0", Scale="1,1,1") {
    alphaInputType=Bool
    rotationMode=BMM_Additive
    rotationSpace=BCS_BoneSpace
    scaleMode=BMM_Replace
    scaleSpace=BCS_BoneSpace
    translationMode=BMM_Additive
    translationSpace=BCS_BoneSpace
    alphaBoolBlend {blendInTime=0.2, blendOption=Linear, blendOutTime=2}
    boneToModify {boneName=dyn_vest_back_base_r}
    rotation {yaw=-5}
   }
   n47 = "Transform (Modify) Bone - Bone: dyn_vest_back_base_l"(ComponentPose=@n46, bAlphaBoolEnabled=@n40, Translation="0,0,0", Rotation="0,-5,0", Scale="1,1,1") {
    alphaInputType=Bool
    rotationMode=BMM_Additive
    rotationSpace=BCS_BoneSpace
    scaleMode=BMM_Replace
    scaleSpace=BCS_BoneSpace
    translationMode=BMM_Additive
    translationSpace=BCS_BoneSpace
    alphaBoolBlend {blendInTime=0.2, blendOption=Linear, blendOutTime=2}
    boneToModify {boneName=dyn_vest_back_base_l}
    rotation {yaw=-5}
   }
   n48 = "Transform (Modify) Bone - Bone: dyn_vest_back_base_c"(ComponentPose=@n47, bAlphaBoolEnabled=@n40, Translation="0,0,0", Rotation="0,-1,0", Scale="1,1,1") {
    alphaInputType=Bool
    rotationMode=BMM_Additive
    rotationSpace=BCS_BoneSpace
    scaleMode=BMM_Replace
    scaleSpace=BCS_BoneSpace
    translationMode=BMM_Additive
    translationSpace=BCS_BoneSpace
    alphaBoolBlend {blendInTime=0.2, blendOption=Linear, blendOutTime=2}
    boneToModify {boneName=dyn_vest_back_base_c}
    rotation {yaw=-1}
   }
   Output Pose(Result=Component To Local(ComponentPose=@n48)) {layerGroup=None}
   Get bIsCrouchSprinting()
   Get LeftLeg() -> (LeftLeg:double)
   Get RightLeg() -> (RightLeg:double)
  ]}
 EventGraph {nodes [
   Event BlueprintUpdateAnimation()
   > n1 = Sequence!(execute) -> (then_0, then_1, then_2:Exec)
   > Set LeftLeg(LeftLeg:double=GetAlphaFromReferenceTransformDelta(
     SkeletalMeshComponent=Get Skel Mesh()
     BoneName=thigh_l
     BaseName=pelvis
     DeltaComponentType=EulerX
     InputRange=(X=6,Y=-17)
     TwistAxis="1,0,0"
    )) -> (Output_Get:double)
   Set RightLeg(
    execute=@n1.then_1
    RightLeg:double=GetAlphaFromReferenceTransformDelta(
     SkeletalMeshComponent=Get Skel Mesh()
     BoneName=thigh_r
     BaseName=pelvis
     DeltaComponentType=EulerX
     InputRange=(X=15,Y=-5)
     TwistAxis="1,0,0"
    )
   ) -> (Output_Get:double)
   Set Vest(
    execute=@n1.then_2
    Vest:double=GetAlphaFromReferenceTransformDelta(
     SkeletalMeshComponent=Get Skel Mesh()
     BoneName=spine_01
     BaseName=pelvis
     DeltaComponentType=EulerX
     InputRange=(X=1,Y=-10)
     TwistAxis="1,0,0"
    )
   ) -> (Output_Get:double)
   Event BlueprintInitializeAnimation()
   > Set Skel Mesh(Skel Mesh=GetOwnerSkeletalMeshComponent())
  ]}
]
variables [
 LeftLeg {type=number}
 Raw Output {type=number}
 RightLeg {type=number}
 Skel Mesh {type=SkeletalMeshComponent}
 Vest {type=number}
]
