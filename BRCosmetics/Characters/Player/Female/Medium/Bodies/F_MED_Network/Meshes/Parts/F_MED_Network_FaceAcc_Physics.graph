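# Physics asset for the face accessory mesh of the female medium Network character body, covering the head and dynamic hair bones.
# A large head capsule anchors a small Dyn_Master_Hair capsule through a constraint with free swing and twist.
# Five hair strand capsules each constrain directly to that master bone, forming a star rather than chains, with narrow angular limits that keep twist minimal.
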PhysicsAsset /BRCosmetics/Characters/Player/Female/Medium/Bodies/F_MED_Network/Meshes/Parts/F_MED_Network_FaceAcc_Physics
bodies {
 Dyn_Hair10 [{shapeType=Capsule, length=2.91, radius=1, shapeName=Dyn_Hair10_capsule, center=<2.7367 0.4099 -0.1228>, rotation=<rot 0 -74.2813 -88.388>}]
 Dyn_Hair11 [{shapeType=Capsule, length=3.6336, radius=1, shapeName=Dyn_Hair11_capsule, center=<3.3459 0.6098 -0.2807>, rotation=<rot 0 -73.7913 -84.2217>}]
 Dyn_Hair12 [{shapeType=Capsule, length=3.8591, radius=1, shapeName=Dyn_Hair12_capsule, center=<3.5923 0.1034 0.3461>, rotation=<rot 0 106.0171 -83.9945>}]
 Dyn_Hair2 [{shapeType=Capsule, length=3.3626, radius=1, shapeName=Dyn_Hair2_capsule, center=<3.0499 0.0728 0.7356>, rotation=<rot 0 90.9664 -68.9544>}]
 Dyn_Hair6 [{shapeType=Capsule, length=2.1624, radius=1, shapeName=Dyn_Hair6_capsule, center=<1.9867 0.1364 -0.1237>, rotation=<rot 0 -83.7784 -81.4657>}]
 Dyn_Master_Hair [{shapeType=Capsule, length=1, radius=1, shapeName=Dyn_Master_Hair_capsule}]
 head [{shapeType=Capsule, length=6, radius=9, shapeName=head_capsule, center=<0.944 0.5259 -0.0001>, rotation=<rot 90 0 -15.0003>}]
}
constraints [
 {bone1Name=Dyn_Hair2, bone2Name=Dyn_Master_Hair, swing1LimitDegrees=10, swing2LimitDegrees=1, twistLimitDegrees=1}
 {bone1Name=Dyn_Hair6, bone2Name=Dyn_Master_Hair, swing1LimitDegrees=20, swing2LimitDegrees=10, twistLimitDegrees=1}
 {bone1Name=Dyn_Hair10, bone2Name=Dyn_Master_Hair, swing1LimitDegrees=20, swing2LimitDegrees=5, twistLimitDegrees=1}
 {bone1Name=Dyn_Hair11, bone2Name=Dyn_Master_Hair, swing1LimitDegrees=20, swing2LimitDegrees=5, twistLimitDegrees=1}
 {bone1Name=Dyn_Hair12, bone2Name=Dyn_Master_Hair, swing1LimitDegrees=20, swing2LimitDegrees=5, twistLimitDegrees=1}
 {
  bone1Name=Dyn_Master_Hair
  bone2Name=head
  swing1LimitDegrees=45
  swing1Motion=Free
  swing2LimitDegrees=45
  swing2Motion=Free
  twistLimitDegrees=45
  twistMotion=Free
 }
]
